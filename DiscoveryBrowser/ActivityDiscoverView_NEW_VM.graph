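# Discover view widget for the activity browser: a vertical list of homebar, search/tab, grid, horizontal-scrolling and other rows, built on FortActivityDiscoverViewV2.
# Generated rows get delegates bound by type, relaying clicks, tab changes and navigation through the view's dispatchers.
# A throbber shows until the activity query finishes, then the list intro plays.
# Fullscreen changes update rows and transitions; helpers handle analytics and nested-surface setup.
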WidgetBlueprint /DiscoveryBrowser/ActivityDiscoverView_NEW_VM
animations [
 Intro {
  binding_count=1
  length_seconds=0.65
  object=&self:Intro
  sequence {
   bindings [BrowserList_Activities {
     id=#1
     possessed_class=FortActivityBrowserListView
     tracks [
      {
       type=MovieSceneFloatTrack
       display_name=Render Opacity
       sections [{type=MovieSceneFloatSection, end_frame=65, start_frame=0, channels [None {type=MovieSceneScriptingFloatChannel, keys [[0,0],[55,1]]}]}]
      }
      {
       type=MovieScene2DTransformTrack
       sections [{
         type=MovieScene2DTransformSection
         end_frame=65
         start_frame=0
         channels [Translation.Y {type=MovieSceneScriptingFloatChannel, keys [[0,40],[65,0]]}]
        }]
      }
     ]
    }]
   display_rate {numerator=100}
  }
 }
 ShowAndHideRowLoadingSpinner {
  binding_count=1
  length_seconds=0.32
  object=&self:ShowAndHideRowLoadingSpinner
  sequence {bindings [WBP_UIKit_Throbber {
     id=#2
     possessed_class=WBP_UIKit_Throbber_C
     tracks [
      {
       type=MovieSceneFloatTrack
       display_name=Render Opacity
       sections [{type=MovieSceneFloatSection, channels [None {type=MovieSceneScriptingFloatChannel, keys [[1.6,0],[6.4,1]]}]}]
      }
      {
       type=MovieScene2DTransformTrack
       sections [{
         type=MovieScene2DTransformSection
         channels [
          Scale.X {type=MovieSceneScriptingFloatChannel, keys [[1.6,0.875],[6.4003,1]]}
          Scale.Y {type=MovieSceneScriptingFloatChannel, keys [[1.6,0.875],[6.4003,1]]}
         ]
        }]
      }
      {
       type=MovieSceneByteTrack
       display_name=Visibility
       sections [{type=MovieSceneByteSection, channels [None {type=MovieSceneScriptingByteChannel, keys [[0,1],[1.6,3]]}]}]
      }
     ]
    }]}
 }
]
graphs [
 AllowMouseWheelScrolling {nodes [
   n1 = AllowMouseWheelScrolling!() -> (then, AllowMouseWheelScrolling:bool=~) {is=title K2Node_FunctionEntry}
   > SetWheelScrollMultiplier!(
    execute
    self:List View Base Object Reference=Get BrowserList_Activities() -> (BrowserList_Activities:Fort Activity Browser List View Object Reference)
    NewWheelScrollMultiplier:float=SelectFloat(A=2, B=0, bPickA=@n1)
   ) -> (then)
  ]}
 BroadcastSearchBackButtonClicked {nodes [BroadcastSearchBackButtonClicked!() -> (then, ButtonObject:WBP Bottom Bar Decoy Button Object Reference) {is=title K2Node_FunctionEntry}]}
 BroadcastTabSearchSelectedIndex {nodes [BroadcastTabSearchSelectedIndex!() -> (then, SelectedTabIndex:int) {is=title K2Node_FunctionEntry}]}
 BroadcastToHomespaceTileClicked {nodes [BroadcastToHomespaceTileClicked() {is=title K2Node_FunctionEntry}]}
 EndDiscoverAnalytics {nodes [
   EndDiscoverAnalytics() {is=title K2Node_FunctionEntry}
   > FireDiscoverExitedAnalyticEvent(self:Fort Activity Discover View V2 Object Reference)
   > Branch(Condition=Get IsFullscreen() -> (IsFullscreen:bool))
   n1 = Get Cached Surface View Model() -> (Cached Surface View Model:Discover Surface View Model Object Reference)
   > Branch(Condition=IsValid(Object=@n1))
   > FireDiscoverTabExitedAnalyticEvent(self:Discover Surface View Model Object Reference=@n1)
   > StopVisibleImpressions(self:Discover Surface View Model Object Reference=@n1)
  ]}
 EventGraph {nodes [
   n1 = "On Entry Generated (BrowserList_Activities)"!() -> (OutputDelegate, then, Widget:User Widget Object Reference) {is=title K2Node_ComponentBoundEvent}
   > n2 = Cast To ActivityBrowserDiscoverItemRow_Homebar!(execute, Object=@n1.Widget) -> (then, CastFailed, AsActivity Browser Discover Item Row Homebar:Activity Browser Discover Item Row Homebar Object Reference) {is=title K2Node_DynamicCast}
   > Bind Event to On Row Broadcast Click to Discover(self:Activity Browser Discover Item Row Homebar Object Reference=@n2, Delegate=Create Event()) {is=title K2Node_AddDelegate}
   n3 = Cast To WBP_SearchWithTabs_Wrapper!(execute=@n2.CastFailed, Object=@n1.Widget) -> (then, CastFailed, AsWBP Search with Tabs Wrapper:WBP Search With Tabs Wrapper Object Reference) {is=title K2Node_DynamicCast}
   n4 = Get WBP_Row_InputFieldWithTabs!(self:WBP Search With Tabs Wrapper Object Reference=@n3) -> (WBP_Row_InputFieldWithTabs:WBP Row Input Field With Tabs Object Reference)
   > Bind Event to On List Scrolled Out(self:WBP Row Input Field With Tabs Object Reference=@n4, Delegate=Create Event()) {is=title K2Node_AddDelegate}
   > Bind Event to On Back Button Hack Clicked(self:WBP Row Input Field With Tabs Object Reference=@n4, Delegate=Create Event()) {is=title K2Node_AddDelegate}
   n5 = Create Event()
   > Bind Event to Broadcast Tab Search Selected Index(self:WBP Row Input Field With Tabs Object Reference=@n4, Delegate=@n5) {is=title K2Node_AddDelegate}
   > BindScrollOutEvent!(
    execute
    self:WBP Row Input Field With Tabs Object Reference=@n4
    ActivityDiscoverView:Activity Discover View NEW VM Object Reference=Self-Reference()
    Bind:bool=true
   ) -> (then)
   n6 = Create Event()
   > Bind Event to On Request Navigate to Discover(self:WBP Row Input Field With Tabs Object Reference=@n4, Delegate=@n6) {is=title K2Node_AddDelegate}
   > Set Cached Surface View Model(Cached Surface View Model:Discover Surface View Model Object Reference=Get FortDiscoverSurfaceViewModel!(self:WBP Row Input Field With Tabs Object Reference=@n4) -> (FortDiscoverSurfaceViewModel:Discover Surface View Model Object Reference)) -> (Output_Get:Discover Surface View Model Object Reference)
   n7 = Cast To ActivityBrowserDiscoverGridRowMax!(execute=@n3.CastFailed, Object=@n1.Widget) -> (then, CastFailed, AsActivity Browser Discover Grid Row Max:Activity Browser Discover Grid Row Max Object Reference) {is=title K2Node_DynamicCast}
   > Bind Event to On Row Selection Changed(self:Activity Browser Discover Grid Row Max Object Reference=@n7, Delegate=Create Event()) {is=title K2Node_AddDelegate}
   > UpdateRowPadding!(
    execute
    self:WBP Row Panel List Object Reference=Get WBP_Row_PanelList!(self:Activity Browser Discover Grid Row Max Object Reference=@n7) -> (WBP_Row_PanelList:WBP Row Panel List Object Reference)
    IsTopTenRow:bool=Get bIsTopTenRow!(self:Discover Provider View Model Object Reference=Get FortDiscoverProviderViewModel!(self:Activity Browser Discover Grid Row Max Object Reference=@n7) -> (FortDiscoverProviderViewModel:Discover Provider View Model Object Reference)) -> (bIsTopTenRow:bool)
   ) -> (then)
   n8 = Cast To ActivityBrowserDiscoverRow_HorizontalScrolling!(execute=@n7.CastFailed, Object=@n1.Widget) -> (then, CastFailed, AsActivity Browser Discover Row Horizontal Scrolling:Activity Browser Discover Row Horizontal Scrolling Object Reference) {is=title K2Node_DynamicCast}
   n9 = Get WBP_Row_HorizontalScrollingList!(self:Activity Browser Discover Row Horizontal Scrolling Object Reference=@n8) -> (WBP_Row_HorizontalScrollingList:WBP Row Horizontal Scrolling List Object Reference)
   n10 = Get bIsTopTenRow(self=Get FortDiscoverProviderViewModel(self:Activity Browser Discover Row Horizontal Scrolling Object Reference=@n8))
   > UpdateRowPadding(self:WBP Row Horizontal Scrolling List Object Reference=@n9, IsTopTenRow=@n10)
   > SetLeftArrowPadding!(execute, self:WBP Row Horizontal Scrolling List Object Reference=@n9, IsTop10:bool=@n10) -> (then)
   n11 = "On Entry Released (BrowserList_Activities)"!() -> (OutputDelegate, then, Widget:User Widget Object Reference) {is=title K2Node_ComponentBoundEvent}
   > n12 = Cast To ActivityBrowserDiscoverItemRow_Homebar(Object=@n11.Widget)
   > Unbind all Events from On Row Broadcast Click to Discover!(execute:Exec, self:Activity Browser Discover Item Row Homebar Object Reference=@n12) -> (then:Exec) {is=title K2Node_ClearDelegate}
   n13 = Cast To WBP_SearchWithTabs_Wrapper(execute=@n12.CastFailed, Object=@n11.Widget)
   n14 = Get WBP_Row_InputFieldWithTabs(self=@n13)
   > Unbind all Events from On List Scrolled Out!(execute:Exec, self:WBP Row Input Field With Tabs Object Reference=@n14) -> (then:Exec) {is=title K2Node_ClearDelegate}
   > Unbind Event from On Back Button Hack Clicked(self:WBP Row Input Field With Tabs Object Reference=@n14, Delegate=Create Event()) {is=title K2Node_RemoveDelegate}
   > Unbind Event from Broadcast Tab Search Selected Index(self:WBP Row Input Field With Tabs Object Reference=@n14, Delegate=@n5) {is=title K2Node_RemoveDelegate}
   > BindScrollOutEvent(self=@n14, ActivityDiscoverView=Self-Reference(), Bind=false)
   > Unbind Event from On Request Navigate to Discover(self:WBP Row Input Field With Tabs Object Reference=@n14, Delegate=@n6) {is=title K2Node_RemoveDelegate}
   n15 = Cast To ActivityBrowserDiscoverGridRowMax(execute=@n13.CastFailed, Object=@n11.Widget)
   > Unbind all Events from On Row Selection Changed!(execute:Exec, self:Activity Browser Discover Grid Row Max Object Reference=@n15) -> (then:Exec) {is=title K2Node_ClearDelegate}
   n16 = OnTransitionDiscoverFullscreen!() -> (OutputDelegate, then, IsDiscoverFullscreen:bool=~) {is=title K2Node_CustomEvent}
   > n17 = Set IsFullscreen(IsFullscreen:bool=@n16.IsDiscoverFullscreen) -> (Output_Get:bool)
   > n18 = Branch(Condition=@n17)
   > Call On Transition Discover In!(execute:Exec, self:Activity Discover View NEW VM Object Reference) -> (then:Exec) {is=title K2Node_CallDelegate}
   n19 = GetOwningPlayer()
   > Branch(Condition=AND Boolean(A=IsDiscoveryReduxEnabled!() -> (bIsEnabled:bool), B=IsUsingTouch(PlayerController=@n19)))
   > ScrollListViewToNextRow!(execute, self:Fort Activity Browser List View Object Reference=Get BrowserList_Activities()) -> (then, ReturnValue:bool)
   Call On Transition Discover Out!(execute:Exec=@n18.else, self:Activity Discover View NEW VM Object Reference) -> (then:Exec) {is=title K2Node_CallDelegate}
   OnReceivedTileClickFromHomebarRow() {is=title K2Node_CustomEvent}
   > Call Broadcast to Homespace Tile Clicked!(execute:Exec, self:Activity Discover View NEW VM Object Reference) -> (then:Exec) {is=title K2Node_CallDelegate}
   OnSearchRowListScrolledOut() {is=title K2Node_CustomEvent}
   > Call Navigate Up Out Of Search Row!(execute:Exec, self:Activity Discover View NEW VM Object Reference) -> (then:Exec) {is=title K2Node_CallDelegate}
   n20 = OnGridRowSelectionChanged!() -> (OutputDelegate, then, Row:Activity Browser Discover Grid Row Max Object Reference, IsSelected:bool=~, IsViewAllSelected:bool=~) {is=title K2Node_CustomEvent}
   > n21 = Branch(Condition=@n20.IsSelected)
   n22 = Get WBP_Row_PanelList(self=@n20.Row)
   > Branch(Condition=AND Boolean(
     A=Get WasViewAllSelected() -> (WasViewAllSelected:bool)
     B=HasViewAllButton!(self:WBP Row Panel List Object Reference=@n22) -> (HasViewAll:bool)
    ))
   > Set IsViewAllSelected!(execute, IsViewAllSelected:bool=true, self:WBP Row Panel List Object Reference=@n22) -> (then, Output_Get:bool)
   > SetFocus(self=Get Button_ViewAll_PaddingHalfOfTileEntrySpacing!(self:WBP Row Panel List Object Reference=@n22) -> (Button_ViewAll_PaddingHalfOfTileEntrySpacing:WBP Discover Button Generic Object Reference))
   Set WasViewAllSelected(execute=@n21.else, WasViewAllSelected:bool=@n20.IsViewAllSelected) -> (Output_Get:bool)
   n23 = OnSearchTabSelectedIndex!() -> (OutputDelegate, then, SelectedTabIndex:int) {is=title K2Node_CustomEvent}
   > Call Broadcast Tab Search Selected Index!(execute:Exec, self:Activity Discover View NEW VM Object Reference, SelectedTabIndex:int=@n23.SelectedTabIndex) -> (then:Exec) {is=title K2Node_CallDelegate}
   n24 = OnBottomHackButtonBackClicked!() -> (OutputDelegate, then, ButtonObject:WBP Bottom Bar Decoy Button Object Reference) {is=title K2Node_CustomEvent}
   > Call Broadcast Search Back Button Clicked!(
    execute:Exec
    self:Activity Discover View NEW VM Object Reference
    ButtonObject:WBP Bottom Bar Decoy Button Object Reference=@n24.ButtonObject
   ) -> (then:Exec) {is=title K2Node_CallDelegate}
   On Request Navigate To Discover() {is=title K2Node_CustomEvent}
   > n25 = Branch(Condition=Get IsFullscreen())
   ScrollListViewToNextRow(execute=@n25.else, self=Get BrowserList_Activities())
   Event Construct()
   > Set HasIntroPlayed(HasIntroPlayed:bool=false) -> (Output_Get:bool)
   > Set IsFullscreen(IsFullscreen=false)
   > ResetScreenState()
   > PlayAnimationForward(InAnimation=Get ShowAndHideRowLoadingSpinner() -> (ShowAndHideRowLoadingSpinner:Widget Animation Object Reference))
   > BindToAnimationFinished()
   > Parent: Construct()
   n26 = Event OnListViewScrolled!() -> (OutputDelegate, then, ItemOffset:float, DistanceRemaining:float) {is=title K2Node_Event}
   > Call On List Scrolled!(
    execute:Exec
    self:Activity Discover View NEW VM Object Reference
    ItemOffset:double=@n26.ItemOffset
    DistanceRemaining:double=@n26.DistanceRemaining
   ) -> (then:Exec) {is=title K2Node_CallDelegate}
   Event OnActivityUpdated() {is=title K2Node_Event}
   > Call On Home Bar Updated!(
    execute:Exec
    self:Activity Discover View NEW VM Object Reference
    IsFirstRow:bool="Equal (Integer)"(A=Get RowIndex() -> (RowIndex:int=0), B=0) {is=title K2Node_CallFunction}
   ) -> (then:Exec) {is=title K2Node_CallDelegate}
   Event OnInitialized()
   > n27 = Branch(Condition=IsDiscoveryReduxEnabled())
   > Set CustomKeyInputAction(CustomKeyInputAction:Data Table Row Handle Structure=Make Data Table Row Handle!(DataTable:Data Table Object Reference=/FortUILibrary/Foundation/ActivityBrowserInputActionTable, RowName:Name=CustomMatchmakingModal_Redux) -> (DataTableRowHandle:Data Table Row Handle Structure) {is=title K2Node_MakeStruct}) -> (Output_Get:Data Table Row Handle Structure)
   > SetIsTouchScrollingEnabled!(execute, self:List View Base Object Reference=Get BrowserList_Activities(), bInIsTouchScrollingEnabled:bool=false) -> (then)
   Set CustomKeyInputAction(
    execute=@n27.else
    CustomKeyInputAction=Make Data Table Row Handle(DataTable=&ActivityBrowserInputActionTable, RowName=CustomMatchmakingModal)
   )
   Event OnQueryActivitiesFinished() {is=title K2Node_Event}
   > PlayAnimation(InAnimation=Get Intro())
   > PlayAnimationReverse(InAnimation=Get ShowAndHideRowLoadingSpinner())
   > Call On Play Intro!(execute:Exec, self:Activity Discover View NEW VM Object Reference) -> (then:Exec) {is=title K2Node_CallDelegate}
   > Set HasIntroPlayed(HasIntroPlayed=true)
   > SetVisibility(self=Get Overlay_Nav() -> (Overlay_Nav:Overlay Object Reference), InVisibility=SelfHitTestInvisible)
   n28 = Event OnRowChanged!() -> (OutputDelegate, then, NewCategoryIndex:int=0) {is=title K2Node_Event}
   > Set RowIndex(RowIndex:int=@n28.NewCategoryIndex) -> (Output_Get:int=0)
   > Call On Row Updated!(
    execute:Exec
    self:Activity Discover View NEW VM Object Reference
    IsPromoted:bool=IsShowingPromotedContent!(self:Fort Activity Discover View V2 Object Reference) -> (ReturnValue:bool)
    IsFirstRow:bool="Equal (Integer)"(A=@n28.NewCategoryIndex, B=0)
   ) -> (then:Exec) {is=title K2Node_CallDelegate}
   > Set WasViewAllSelected(WasViewAllSelected=false)
   n29 = Event OnDiscoverFullscreenChanged!() -> (OutputDelegate, then, IsFullscreen:bool) {is=title K2Node_Event}
   > n30 = GetDisplayedEntryWidgets!(execute, self:List View Base Object Reference=Get BrowserList_Activities()) -> (then, ReturnValue:Array of User Widget Object References)
   > n31 = For Each Loop(Array:Array of User Widget Object References=@n30) -> (Array Element:User Widget Object Reference)
   > n32 = Cast To WBP_Row_Grid_Base!(execute, Object=@n31) -> (then, CastFailed, AsWBP Row Grid Base:WBP Row Grid Base Object Reference) {is=title K2Node_DynamicCast}
   > SetDiscoverFullscreenState!(execute, self:WBP Row Grid Base Object Reference=@n32, IsFullscreen:bool=@n29.IsFullscreen) -> (then)
   > n33 = Cast To WBP_SearchWithTabs_Wrapper(Object=@n31)
   > n34 = Branch(Condition=@n29.IsFullscreen)
   > ActivateWidget(self=@n33)
   DeactivateWidget(execute=@n34.else, self=@n33)
   n35 = Event ConfigureNestedSurface!() -> (
    OutputDelegate
    then
    SurfaceWidget:Common Activatable Widget Object Reference
    SurfaceTileItemVM:Fort Discover Surface Tile Item VM Object Reference
   ) {is=title K2Node_Event}
   > n36 = Cast To WBP_Discover_NestedSurface!(execute, Object=@n35.SurfaceWidget) -> (then, CastFailed, AsWBP Discover Nested Surface:WBP Discover Nested Surface Object Reference) {is=title K2Node_DynamicCast}
   > SetFortDiscoverSurfaceTileItemVM!(
    execute
    self:WBP Discover Nested Surface Object Reference=@n36
    Viewmodel:Fort Discover Surface Tile Item VM Object Reference=@n35.SurfaceTileItemVM
   ) -> (then)
   > On Disover Surface Tile Item VM Set(self:WBP Discover Nested Surface Object Reference=@n36)
   > SetSurfaceName!(
    execute
    self:Discover Surface View Model Object Reference=Get FortDiscoverSurfaceViewModel(self:WBP Discover Nested Surface Object Reference=@n36)
    InSurfaceName:Name=GetReferenceSurfaceName!(self:Fort Discover Surface Tile Item VM Object Reference=@n35.SurfaceTileItemVM) -> (ReturnValue:Name=None)
   ) -> (then)
   n37 = Sequence()
   Do Once(Reset=@n37.then_0)
   > n38 = CreateSound2D!(
    execute
    Sound:Sound Base Object Reference=Get AmbientBackgroundLoop() -> (AmbientBackgroundLoop:Sound Base Object Reference)
    VolumeMultiplier:float=1
    PitchMultiplier:float=1
    StartTime:float=0
    ConcurrencySettings:Sound Concurrency Object Reference
    bPersistAcrossLevelTransition:bool=false
    bAutoDestroy:bool=true
   ) -> (then, ReturnValue:Audio Component Object Reference)
   > n39 = Set BackgroundAmbience(BackgroundAmbience:Audio Component Object Reference=@n38) -> (Output_Get:Audio Component Object Reference)
   n40 = Get MusicEventSubsystem() -> (ReturnValue:Music Event Subsystem Object Reference) {is=title K2Node_GetSubsystem}
   > AddEventByComponent!(
    execute
    self:Music Event Subsystem Object Reference=@n40
    Component:Audio Component Object Reference=@n39
    EventTag:Gameplay Tag Structure=(TagName="MusicEventSubsystem.Event.Lobby.MP")
    EventBehaviorTags:Gameplay Tag Container Structure=(GameplayTags=)
    MusicClock:Music Clock Object Reference
    GroupName:Name
   ) -> (then)
   > Play!(execute, self:Audio Component Object Reference=@n39, StartTime:float=0) -> (then)
   n41 = Get ReverbSubmixEffect() -> (ReverbSubmixEffect:Submix Effect Dynamic Reverb Preset Object Reference)
   n42 = Get MusicSubmix() -> (MusicSubmix:Sound Submix Object Reference)
   > AddSubmixEffect!(execute, SoundSubmix:Sound Submix Object Reference=@n42, SubmixEffectPreset:Sound Effect Submix Preset Object Reference=@n41) -> (then, ReturnValue:int=0)
   n43 = Get ControlBusMix() -> (ControlBusMix:Sound Control Bus Mix Object Reference)
   > Activate Control Bus Mix!(execute, Mix:Sound Control Bus Mix Object Reference=@n43) -> (then)
   n44 = Get!(execute:Exec=@n37.then_1) -> (then:Exec, else:Exec, BackgroundAmbience:Audio Component Object Reference)
   > Execute Trigger Parameter!(execute, self:Audio Parameter Controller Interface Interface=@n44, InName:Name=Stop) -> (then)
   > RemoveEventCategory!(execute, self:Music Event Subsystem Object Reference=@n40, EventTag:Gameplay Tag Structure) -> (then)
   > Deactivate Control Bus Mix!(execute, Mix:Sound Control Bus Mix Object Reference=@n43) -> (then)
   > RemoveSubmixEffect!(execute, SoundSubmix:Sound Submix Object Reference=@n42, SubmixEffectPreset:Sound Effect Submix Preset Object Reference=@n41) -> (then)
   Reroute Node(InputPin:Wildcard) -> (OutputPin:Wildcard)
  ]}
 GetHasIntroPlayed {nodes [
   GetHasIntroPlayed() {is=title K2Node_FunctionEntry}
   > Return Node!(execute, IsFinishedLoading:bool=Get HasIntroPlayed() -> (HasIntroPlayed:bool))
  ]}
 GetIndexOfCurrentActivity {nodes [
   GetIndexOfCurrentActivity() {is=title K2Node_FunctionEntry}
   n1 = GetDisplayedEntryWidgets!(self=Get BrowserList_Activities()) -> (ReturnValue)
   > Branch(Condition=Is Not Empty!(TargetArray:Array of User Widget Object References=@n1) -> (ReturnValue:bool))
   > n2 = For Each Loop with Break!(Exec:Exec, Array:Array of User Widget Object References=@n1, Break:Exec=@n3.then) -> (LoopBody:Exec, Array Element:User Widget Object Reference, Array Index:int, Completed:Exec)
   > n3 = Cast To ActivityBrowserDiscoverItemRow_Homebar(Object=@n2)
   Branch(execute=@n2.Completed, Condition=IsValid(Object=@n3))
   n4 = GetDisplayedEntryWidgets(self=Get ListView_Tiles!(self:Fort Discover Item Browser Row Object Reference=@n3) -> (ListView_Tiles:Fort Discover Item List View Object Reference))
   > Branch(Condition=Is Not Empty(TargetArray=@n4))
   > n5 = For Each Loop with Break(Array=@n4, Break=@n8.then)
   > n6 = Cast To ActivityBrowserTile_Homebar_NEW_VM!(execute, Object=@n5) -> (then, CastFailed, AsActivity Browser Tile Homebar NEW VM:Activity Browser Tile Homebar NEW VM Object Reference) {is=title K2Node_DynamicCast}
   > n7 = Cast To WBP_IslandTile_Minimal!(
    execute
    Object=Get Display_TileDetails!(self:Fort Activity Browser Tile Object Reference=@n6) -> (Display_TileDetails:Fort Activity Tile Details Display Object Reference)
   ) -> (then, CastFailed, AsWBP Island Tile Minimal:WBP Island Tile Minimal Object Reference) {is=title K2Node_DynamicCast}
   > n8 = Branch(Condition=Get bIsCurrentActivity!(self:Discover Activity View Model Object Reference=Get FortActivityViewModel!(self:WBP Island Tile Minimal Object Reference=@n7) -> (FortActivityViewModel:Discover Activity View Model Object Reference)) -> (bIsCurrentActivity:bool))
   Return Node!(execute=@n5.Completed, TileIndex:int=@n5.Array Index)
  ]}
 GetNestedSurfaceClass {nodes [
   GetNestedSurfaceClass() {is=title K2Node_FunctionEntry}
   > Return Node!(
    execute
    ReturnValue:Common Activatable Widget Class Reference=Get NestedSurfaceClass() -> (NestedSurfaceClass:Common Activatable Widget Class Reference)
   )
  ]}
 IsDiscoveryReduxEnabled {nodes [
   IsDiscoveryReduxEnabled() {is=title K2Node_FunctionEntry}
   > Return Node!(
    execute
    bIsEnabled:bool=GetDiscoveryReduxExperimentValue!(FortPlayerController:Player Controller Object Reference=GetOwningPlayer()) -> (ReturnValue:bool)
   )
  ]}
 NavigateUpOutOfSearchRow {nodes [NavigateUpOutOfSearchRow() {is=title K2Node_FunctionEntry}]}
 OnHomeBarUpdated {nodes [OnHomeBarUpdated!() -> (then, IsFirstRow:bool=~) {is=title K2Node_FunctionEntry}]}
 OnListScrolled {nodes [OnListScrolled!() -> (then, ItemOffset:double=~, DistanceRemaining:double=~) {is=title K2Node_FunctionEntry}]}
 OnOverSwipeUp {nodes [OnOverSwipeUp() {is=title K2Node_FunctionEntry}]}
 OnPlayIntro {nodes [OnPlayIntro() {is=title K2Node_FunctionEntry}]}
 OnRowUpdated {nodes [OnRowUpdated!() -> (then, IsPromoted:bool=~, IsFirstRow:bool=~) {is=title K2Node_FunctionEntry}]}
 OnSwipeDownWhenCollapsed {nodes [OnSwipeDownWhenCollapsed() {is=title K2Node_FunctionEntry}]}
 OnTransitionDiscoverIn {nodes [OnTransitionDiscoverIn() {is=title K2Node_FunctionEntry}]}
 OnTransitionDiscoverOut {nodes [OnTransitionDiscoverOut() {is=title K2Node_FunctionEntry}]}
 ResetScreenState {nodes [
   ResetScreenState() {is=title K2Node_FunctionEntry}
   > Set IsImageLoadingSpinnerActive(IsImageLoadingSpinnerActive:bool=false) -> (Output_Get:bool)
   > SetRenderOpacity(self=Get BrowserList_Activities(), InOpacity=0)
   n1 = Get WBP_UIKit_Throbber() -> (WBP_UIKit_Throbber:WBP UIKit Throbber Object Reference)
   > SetRenderOpacity(self=@n1, InOpacity=0)
   > SetVisibility(self=@n1, InVisibility=Collapsed)
   > SetVisibility(self=Get Overlay_Nav(), InVisibility=HitTestInvisible)
   > n2 = For Each Loop(Array:Array of Image Object References=Get KeyArt_ImageWidgets() -> (KeyArt_ImageWidgets:Array of Image Object References)) -> (Array Element:Image Object Reference)
   n3 = Get Intro()
   SetAnimationCurrentTime!(
    execute=@n2.Completed
    self:User Widget Object Reference
    InAnimation:Widget Animation Object Reference=@n3
    InTime:float=GetStartTime!(self:Widget Animation Object Reference=@n3) -> (ReturnValue:float)
   ) -> (then)
   n4 = GetDynamicMaterial(execute=@n2.LoopBody, self=@n2)
   > SetScalarParameterValue(self=@n4, ParameterName=ImageScaleAdd_Animate, Value=0)
   > SetScalarParameterValue(self=@n4, ParameterName=NoImage_Animate, Value=0)
   > SetTextureParameterValue(self=@n4, ParameterName=ModeKeyArt, Value=/Game/UI/DiscoveryShared/Art/T_UI_Discovery_DefaultEmpty)
  ]}
 SnapToEndOfAnimation {nodes [
   n1 = SnapToEndOfAnimation!() -> (then, InAnimation:Widget Animation Object Reference, PlayMode:EUMGSequencePlayMode Enum=Forward) {is=title K2Node_FunctionEntry}
   > PlayAnimation(InAnimation=@n1, StartAtTime=GetEndTime(self=@n1), PlayMode=@n1.PlayMode)
  ]}
 StartDiscoverAnalytics {nodes [
   StartDiscoverAnalytics() {is=title K2Node_FunctionEntry}
   > StartDiscoverAnalyticSession(self:Fort Activity Discover View V2 Object Reference)
   > Branch(Condition=Get IsFullscreen())
   n1 = Get Cached Surface View Model()
   > Branch(Condition=IsValid(Object=@n1))
   > FireDiscoverTabEnteredAnalyticEvent(self:Discover Surface View Model Object Reference=@n1)
  ]}
]
variables [
 AmbientBackgroundLoop {type=SoundBase, default=/Game/Sounds/Fort_Music/Music_MP_FrontEnd/wav/MS_MultiProduct_Amb_Bed}
 Athena Lobby {type=AthenaLobbyBase, default=null}
 BackgroundAmbience {type=AudioComponent, default=null}
 BroadcastSearchBackButtonClicked {}
 BroadcastTabSearchSelectedIndex {}
 BroadcastToHomespaceTileClicked {}
 Cached Surface View Model {type=FortDiscoverSurfaceViewModel, default=null}
 ControlBusMix {type=SoundControlBusMix, default=/Game/Sounds/Fort_Music/Music_MP_FrontEnd/wav/MP_UI_Amb_GBM}
 DefaultImage {type=Texture, default=&T_UI_Discovery_DefaultEmpty}
 DetailsMaxHeight {type=number, default=576}
 DetailsMaxHeightMobile {type=number, default=480}
 HasIntroPlayed {type=boolean}
 IsFullscreen {type=boolean}
 IsImageLoadingSpinnerActive {type=boolean}
 IsPromotedContentDisplayed {type=boolean}
 KeyArt_ImageWidgets {type="array<Image>", default []}
 KeyArtMaxWidth {type=number, default=1536}
 KeyArtMaxWidthMobile {type=number, default=1300}
 KeyArtParameter {type=string, default=ModeKeyArt}
 MusicSubmix {type=SoundSubmix, default=/Game/Sounds/SubMixes/Fortnite_Music}
 NavigateUpOutOfSearchRow {}
 NestedSurfaceClass {type=Class, default=&/WBP_Discover_NestedSurface.WBP_Discover_NestedSurface_C}
 NextImage {type=Texture, default=null}
 OnHomeBarUpdated {}
 OnListScrolled {}
 OnOverSwipeUp {}
 OnPlayIntro {}
 OnRowUpdated {}
 OnSwipeDownWhenCollapsed {}
 OnTransitionDiscoverIn {}
 OnTransitionDiscoverOut {}
 ReverbSubmixEffect {type=SubmixEffectDynamicReverbPreset, default=/Game/Sounds/SubMixes/FX/MP_Amb_Music_Reverb}
 RowIndex {type=integer}
 WasViewAllSelected {type=boolean}
]
widgets {
 info {
  inheritedWidgetCount=1
  parentClass=/Script/DiscoveryBrowserUI.FortActivityDiscoverViewV2
  rootWidgetClass=&Overlay
  widgetCount=3
 }
 widgets [
  Overlay_Nav {widgetClassPath=&Overlay, bIsVariable=true, widget=Overlay_Nav}
  BrowserList_Activities {
   widgetClassPath=/Script/DiscoveryBrowserUI.FortActivityBrowserListView
   bInherited=true
   bIsVariable=true
   parent=Overlay_Nav
   widget=BrowserList_Activities
   properties {
    bAllowDragging=false
    bEnableRightClickScrolling=false
    bInEnableTouchAnimatedScrolling=true
    bannerItemRowClass=/DiscoveryBrowser/Rows/Billboard/WBP_Discovery_Row_Billboard.WBP_Discovery_Row_Billboard_C
    clipping=Inherit
    discoverItemRowClass=&/ActivityBrowserDiscoverItemRow_NEW_VM.ActivityBrowserDiscoverItemRow_NEW_VM_C
    entryWidgetClass=&/ActivityBrowserRow_NEW_VM.ActivityBrowserRow_NEW_VM_C
    fixedLineScrollOffset=0.4
    homebarItemRowClass=&/ActivityBrowserDiscoverItemRow_Homebar.ActivityBrowserDiscoverItemRow_Homebar_C
    nonScrollingItemRowClass=&/ActivityBrowserDiscoverGridRowMax.ActivityBrowserDiscoverGridRowMax_C
    numDesignerPreviewEntries=4
    renderOpacity=0
    scrollingViewModelItemRowClass=&/ActivityBrowserDiscoverRow_HorizontalScrolling.ActivityBrowserDiscoverRow_HorizontalScrolling_C
    tabBarItemRowClass=/DiscoveryBrowser/Rows/WBP_SearchWithTabs_Wrapper.WBP_SearchWithTabs_Wrapper_C
    visibility=Visible
    wheelScrollMultiplier=2
   }
   slot {class=OverlaySlot, properties {horizontalAlignment=HAlign_Fill, verticalAlignment=VAlign_Fill}}
  }
  WBP_UIKit_Throbber {
   widgetClassPath=/FortUILibrary/UIKit/Throbber/WBP_UIKit_Throbber
   bIsVariable=true
   parent=Overlay_Nav
   widget=WBP_UIKit_Throbber
   slot {class=OverlaySlot, properties {padding=<ltrb 80 120 0 0>}}
  }
 ]
}
