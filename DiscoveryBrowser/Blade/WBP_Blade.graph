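# WBP_Blade is a reusable Discovery Browser blade container with named content slots, anchored to a screen edge or fullscreen, that expands, contracts and peeks.
# Widget animations drive float lerp setters, which set animation state and nudge the grid-slot content offscreen by size-based percentages, with Compact UI variants.
# An optional secondary peek state interpolates between two push amounts.
# Click-catcher and peek buttons, transitions and resizes notify the parent through dispatchers.
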WidgetBlueprint /DiscoveryBrowser/Blade/WBP_Blade
animations [
 Anim_ExpandContract {
  binding_count=2
  length_seconds=0.4
  object=&self:Anim_ExpandContract
  sequence {
   bindings [
    WBP_Blade_C_34 {
     id=#1
     possessed_class=WBP_Blade_C
     tracks [{
       type=MovieSceneDoubleTrack
       display_name=Lerp Animator Expand Contract Blade
       sections [{type=MovieSceneDoubleSection, end_frame=24, start_frame=0, channels [None {type=MovieSceneScriptingDoubleChannel, keys [[0,1],[24,0]]}]}]
      }]
    }
    EmptyButton_ClickCatcherCloseBlade {
     id=#2
     possessed_class=Button
     tracks [{
       type=MovieSceneFloatTrack
       display_name=Render Opacity
       sections [{type=MovieSceneFloatSection, end_frame=24, start_frame=0, channels [None {type=MovieSceneScriptingFloatChannel, keys [[0,0],[24,1]]}]}]
      }]
    }
   ]
   display_rate {numerator=60}
  }
 }
 Anim_BackplateOpacity {
  binding_count=1
  length_seconds=0.35
  object=&self:Anim_BackplateOpacity
  sequence {bindings [FullBackplate_Image {
     id=#3
     possessed_class=Image
     tracks [{
       type=MovieSceneFloatTrack
       display_name=Render Opacity
       sections [{type=MovieSceneFloatSection, end_frame=7, start_frame=0, channels [None {type=MovieSceneScriptingFloatChannel, keys [[0,0],[7,1]]}]}]
      }]
    }]}
 }
 Anim_BetweenPeekStates {
  binding_count=2
  length_seconds=0.3309
  object=&self:Anim_BetweenPeekStates
  sequence {bindings [
    WBP_Blade_C_58 {
     id=#4
     possessed_class=WBP_Blade_C
     tracks [{
       type=MovieSceneDoubleTrack
       display_name=Lerp Animator Between Peek States
       sections [{type=MovieSceneDoubleSection, end_frame=6, start_frame=0, channels [None {type=MovieSceneScriptingDoubleChannel, keys [[0,1],[6.6177,0]]}]}]
      }]
    }
    EmptyButton_ForPeekMode {
     id=#5
     possessed_class=Button
     tracks [{
       type=MovieSceneByteTrack
       display_name=Visibility
       sections [{type=MovieSceneByteSection, end_frame=6, start_frame=0, channels [None {type=MovieSceneScriptingByteChannel, keys [[0,1],[6.6177,0]]}]}]
      }]
    }
   ]}
 }
]
graphs [
 CalculateNudgeValue {nodes [
   n1 = CalculateNudgeValue!() -> (then, PeekValue:double=~) {is=title K2Node_FunctionEntry}
   > n2 = Switch on E_BladeUIAnchor!(execute, Selection:E_BladeUIAnchor Enum=Get EBladeAlignment() -> (EBladeAlignment:E_BladeUIAnchor Enum=NewEnumerator0)) -> (NewEnumerator0:Exec, NewEnumerator1:Exec, NewEnumerator2:Exec, NewEnumerator3:Exec) {is=title K2Node_SwitchEnum}
   Set NudgeForOffscreen(
    execute=@n2.NewEnumerator2
    NudgeForOffscreen:double=float * float(A=Get ContentWidth() -> (ContentWidth:double), B=@n1) {is=title K2Node_PromotableOperator}
   ) -> (Output_Get:double)
   n3 = Negate Float!(" ":Exec=@n2.NewEnumerator0, Value:"Float (double-precision) (by ref)"=float * float(A=Get ContentWidth(), B=@n1)) -> ("  ":Exec, Result:double=~) {is=title K2Node_MacroInstance}
   > Set NudgeForOffscreen(NudgeForOffscreen=@n3)
   Set NudgeForOffscreen(execute=@n2.NewEnumerator3, NudgeForOffscreen=float * float(A=Get ContentHeight() -> (ContentHeight:double), B=@n1))
   n4 = Negate Float(" "=@n2.NewEnumerator1, Value=float * float(A=Get ContentHeight(), B=@n1)) -> (Result=~)
   > Set NudgeForOffscreen(NudgeForOffscreen=@n4)
  ]}
 EventGraph {nodes [
   "On Clicked (EmptyButton_ClickCatcherCloseBlade)"() {is=title K2Node_ComponentBoundEvent}
   > OnBladeContract()
   > Call On Notify Parent Blade Catcher Clicked!(execute:Exec, self:WBP Blade Object Reference) -> (then:Exec) {is=title K2Node_CallDelegate}
   "On Clicked (EmptyButton_ForPeekMode)"() {is=title K2Node_ComponentBoundEvent}
   > Call On Notify Parent Peek Button Clicked!(execute:Exec, self:WBP Blade Object Reference) -> (then:Exec) {is=title K2Node_CallDelegate}
   OnBladeExpand() {is=title K2Node_CustomEvent}
   > n1 = Switch on E_BladeAnimationState!(execute, Selection:E_BladeAnimationState Enum=Get EBladeAnimState() -> (EBladeAnimState:E_BladeAnimationState Enum=NewEnumerator0)) -> (NewEnumerator0:Exec, NewEnumerator1:Exec, NewEnumerator2:Exec, NewEnumerator4:Exec) {is=title K2Node_SwitchEnum}
   Call On Notify Parent Blade Expand Start!(execute:Exec=[@n1.NewEnumerator0, @n1.NewEnumerator1], self:WBP Blade Object Reference) -> (then:Exec) {is=title K2Node_CallDelegate}
   n2 = Get Anim_ExpandContract() -> (Anim_ExpandContract:Widget Animation Object Reference)
   > PlayAnimationForward(InAnimation=@n2)
   > Branch(Condition=Get AnimateBackplateWhenExpanding() -> (AnimateBackplateWhenExpanding:bool))
   n3 = Get Anim_BackplateOpacity() -> (Anim_BackplateOpacity:Widget Animation Object Reference)
   > PlayAnimationForward(InAnimation=@n3)
   OnBladeContract() {is=title K2Node_CustomEvent}
   > n4 = Switch on E_BladeAnimationState(Selection=Get EBladeAnimState() -> (EBladeAnimState=NewEnumerator0))
   Call On Notify Parent Blade Contract Start!(execute:Exec=[@n4.NewEnumerator2, @n4.NewEnumerator4], self:WBP Blade Object Reference) -> (then:Exec) {is=title K2Node_CallDelegate}
   > PlayAnimationReverse(InAnimation=@n2)
   > Branch(Condition=Get AnimateBackplateWhenExpanding())
   > PlayAnimationReverse(InAnimation=@n3)
   TransitionToSecondPeekState() {is=title K2Node_CustomEvent}
   > Branch(Condition=Get EnableSecondaryPeekState() -> (EnableSecondaryPeekState:bool))
   n5 = Get Anim_BetweenPeekStates() -> (Anim_BetweenPeekStates:Widget Animation Object Reference)
   > PlayAnimationReverse(InAnimation=@n5)
   > Set StartedTransitionToSecondPeekState(StartedTransitionToSecondPeekState:bool=true) -> (Output_Get:bool)
   TransitionToFirstPeekState() {is=title K2Node_CustomEvent}
   > Branch(Condition=Get EnableSecondaryPeekState())
   > PlayAnimationForward(InAnimation=@n5)
   > Set StartedTransitionToSecondPeekState(StartedTransitionToSecondPeekState=false)
   OnPeekStateAnimFinished() {is=title K2Node_CustomEvent}
   > Call On Notify Parent Blade Peek State Transition Finished!(execute:Exec, self:WBP Blade Object Reference) -> (then:Exec) {is=title K2Node_CallDelegate}
   OnContractOrExpandAnimFinished() {is=title K2Node_CustomEvent}
   > Call On Notify Parent Blade Transition Finished!(execute:Exec, self:WBP Blade Object Reference) -> (then:Exec) {is=title K2Node_CallDelegate}
   n6 = Refresh Nudge Position() {is=title K2Node_CustomEvent}
   n7 = GetContentOrViewport SizeAndSetNudgeValue!(execute=[@n9.then, @n6.then, @n28.then, @n26.then], self) -> (then, IsSizeZero:bool)
   > n8 = Branch(Condition=@n7)
   n9 = DelayUntilNextTick!(execute=[@n8.then, @n30.then]) -> (then)
   n10 = Branch(
    execute=@n8.else
    Condition="Equal (Enum)"(A:E_BladeAnimationState Enum=Get EBladeAnimState() -> (EBladeAnimState=NewEnumerator0), B:E_BladeAnimationState Enum=NewEnumerator4)
   )
   > SetVisibility(self=Get EmptyButton_ForPeekMode() -> (EmptyButton_ForPeekMode:Button Object Reference), InVisibility=Collapsed)
   > SetNudge!(
    execute
    self:Grid Slot Object Reference=SlotAsGridSlot(Widget=Get Overlay_BladeContent() -> (Overlay_BladeContent:Overlay Object Reference))
    InNudge:Vector 2D Structure=MakeVector2D(X=0, Y=0)
   ) -> (then)
   SetPeekButtonVisibility!(execute=@n10.else, self, HidePeekButton:bool=Get HidePeekButton() -> (HidePeekButton:bool)) -> (then)
   > n11 = Switch on E_BladeUIAnchor(Selection=Get EBladeAlignment() -> (EBladeAlignment=NewEnumerator0))
   n12 = Get NudgeForOffscreen() -> (NudgeForOffscreen:double)
   SetNudge(
    execute=[@n11.NewEnumerator1, @n11.NewEnumerator3]
    self=SlotAsGridSlot(Widget=Get Overlay_BladeContent())
    InNudge=MakeVector2D(X=0, Y=@n12)
   )
   SetNudge(
    execute=[@n11.NewEnumerator0, @n11.NewEnumerator2]
    self=SlotAsGridSlot(Widget=Get Overlay_BladeContent())
    InNudge=MakeVector2D(X=@n12, Y=0)
   )
   Event PreConstruct()
   > n13 = Branch(Condition=Get UseFullscreenBlade() -> (UseFullscreenBlade:bool))
   > Set Row_Content(Row_Content:int=0) -> (Output_Get:int=0)
   > Set RowSpan_Content(RowSpan_Content:int=3) -> (Output_Get:int=0)
   > Set Column_Content(Column_Content:int=0) -> (Output_Get:int=0)
   > n14 = Set ColumnSpan_Content(ColumnSpan_Content:int=3) -> (Output_Get:int=0)
   n15 = SlotAsGridSlot(Widget=Get Overlay_BladeContent())
   SetRowSpan!(
    execute=[@n23.then, @n21.then, @n14.then]
    self:Grid Slot Object Reference=@n15
    InRowSpan:int=Get RowSpan_Content() -> (RowSpan_Content:int=0)
   ) -> (then)
   > SetRow!(execute, self:Grid Slot Object Reference=@n15, InRow:int=Get Row_Content() -> (Row_Content:int=0)) -> (then)
   > SetColumnSpan!(execute, self:Grid Slot Object Reference=@n15, InColumnSpan:int=Get ColumnSpan_Content() -> (ColumnSpan_Content:int=0)) -> (then)
   > SetColumn!(execute, self:Grid Slot Object Reference=@n15, InColumn:int=Get Column_Content() -> (Column_Content:int=0)) -> (then)
   n16 = Get bHasClickCatcherButton() -> (bHasClickCatcherButton:bool)
   n17 = Get bStartExpanded() -> (bStartExpanded:bool)
   > SetVisibility(
    self=Get EmptyButton_ClickCatcherCloseBlade() -> (EmptyButton_ClickCatcherCloseBlade:Button Object Reference)
    InVisibility=Select(Option 0=Collapsed, Option 1=Visible, Index=AND Boolean(A=@n16, B=@n17))
   )
   > SetStyle!(
    execute
    self:Button Object Reference=Get EmptyButton_ClickCatcherCloseBlade()
    InStyle:"Button Style Structure (by ref)"=Select(
     Option 0:"Button Style Structure (by ref)"=Get ClickCatcherButtonStyle() -> (ClickCatcherButtonStyle:Button Style Structure)
     Option 1:"Button Style Structure (by ref)"=Get TransparentButtonStyle() -> (TransparentButtonStyle:Button Style Structure)
     Index=Get bIsClickCatcherButtonTransparent() -> (bIsClickCatcherButtonTransparent:bool)
    ) -> (ReturnValue:"Button Style Structure (by ref)")
   ) -> (then)
   > SetStyle(self=Get EmptyButton_ForPeekMode(), InStyle=Get TransparentButtonStyle())
   > SetPadding!(
    execute
    self:Overlay Slot Object Reference=SlotAsOverlaySlot(Widget=Get SafeZone() -> (SafeZone:Safe Zone Object Reference))
    InPadding:Margin Structure=Get ContentPadding() -> (ContentPadding:Margin Structure)
   ) -> (then)
   > SetBrushFromMaterial(
    self=Get FullBackplate_Image() -> (FullBackplate_Image:Image Object Reference)
    Material=GetDynamicMaterial!(Brush:"Slate Brush Structure (by ref)"=Get BackplateImageBrush() -> (BackplateImageBrush:Slate Brush Structure)) -> (ReturnValue)
   )
   > SetCompactUIStyle() {is=title K2Node_CallFunction}
   n18 = Get bStartExpanded()
   > Branch(Condition=AND Boolean(A=Get AnimateBackplateWhenExpanding(), B="Equal (Boolean)"(A:bool=@n18, B:bool) -> (ReturnValue:bool=~) {is=title K2Node_PromotableOperator}))
   > SetRenderOpacity(self=Get FullBackplate_Image(), InOpacity=0)
   n19 = Switch on E_BladeUIAnchor(execute=@n13.else, Selection=Get EBladeAlignment() -> (EBladeAlignment=NewEnumerator0))
   > n20 = Set Column_Content(Column_Content=0) -> (Output_Get=0)
   Set Row_Content(execute=[@n20.then, @n24.then], Row_Content=0) -> (Output_Get=0)
   > Set RowSpan_Content(RowSpan_Content=3) -> (Output_Get=0)
   > n21 = Set ColumnSpan_Content(ColumnSpan_Content=1) -> (Output_Get=0)
   n22 = Set Row_Content(execute=@n19.NewEnumerator1, Row_Content=0) -> (Output_Get=0)
   Set RowSpan_Content(execute=[@n22.then, @n25.then], RowSpan_Content=1) -> (Output_Get=0)
   > Set Column_Content(Column_Content=0) -> (Output_Get=0)
   > n23 = Set ColumnSpan_Content(ColumnSpan_Content=3) -> (Output_Get=0)
   n24 = Set Column_Content(execute=@n19.NewEnumerator2, Column_Content=2) -> (Output_Get=0)
   n25 = Set Row_Content(execute=@n19.NewEnumerator3, Row_Content=2) -> (Output_Get=0)
   Event OnViewportResized() {is=title K2Node_Event}
   > Delay(Duration=0.1)
   > Call On Notify Blade Resized!(execute:Exec, self:WBP Blade Object Reference) -> (then:Exec) {is=title K2Node_CallDelegate}
   > n26 = Branch(Condition=NOT Boolean(A=Get StartedTransitionToSecondPeekState() -> (StartedTransitionToSecondPeekState:bool)))
   n27 = Get Anim_BetweenPeekStates()
   n28 = PlayAnimation(execute=@n26.else, InAnimation=@n27, StartAtTime=GetEndTime(self=@n27), PlayMode=Reverse)
   Event Destruct()
   > UnbindAllFromAnimationFinished!(execute, self:User Widget Object Reference, Animation:Widget Animation Object Reference=Get Anim_ExpandContract()) -> (then)
   Event Construct()
   > BindToAnimationEvent!(
    execute
    self:User Widget Object Reference
    Animation:Widget Animation Object Reference=Get Anim_ExpandContract()
    Delegate:Delegate=Create Event()
    AnimationEvent:EWidgetAnimationEvent Enum=Finished
    UserTag:Name=None
   ) -> (then)
   > BindToAnimationEvent(Animation=Get Anim_BetweenPeekStates(), Delegate=Create Event(), AnimationEvent=Finished, UserTag=None)
   > n29 = Branch(Condition=Get bStartExpanded())
   > Set EBladeAnimState(EBladeAnimState:E_BladeAnimationState Enum=NewEnumerator4) -> (Output_Get:E_BladeAnimationState Enum=NewEnumerator0)
   n30 = Set EBladeAnimState(execute=@n29.else, EBladeAnimState=NewEnumerator0) -> (Output_Get=NewEnumerator0)
  ]}
 GetContentOrViewport SizeAndSetNudgeValue {nodes [
   GetContentOrViewport SizeAndSetNudgeValue() {is=title K2Node_FunctionEntry}
   n1 = GetContentSize!() -> (ReturnValue:Vector 2D Structure)
   > n2 = Branch(Condition=IsZero2D!(A:"Vector 2D Structure (by ref)"=@n1) -> (ReturnValue:bool))
   > Return Node!(execute, IsSizeZero:bool=true)
   n3 = BreakVector2D(InVec=@n1)
   Set ContentWidth(execute=@n2.else, ContentWidth:double=@n3) -> (Output_Get:double)
   > Set ContentHeight(ContentHeight:double=@n3.Y) -> (Output_Get:double)
   n4 = IsUsingCompactUI()
   n5 = Get PeekPushPercentSecondary() -> (PeekPushPercentSecondary:double)
   n6 = Get PeekPushPercentSecondary_CompactUI() -> (PeekPushPercentSecondary_CompactUI:double)
   n7 = Get PeekPushPercent_CompactUI() -> (PeekPushPercent_CompactUI:double)
   n8 = Get PeekPushPercent() -> (PeekPushPercent:double)
   > CalculateNudgeValue!(
    execute
    self
    PeekValue:double=SelectFloat(A=SelectFloat(A=@n6, B=@n5, bPickA=@n4), B=SelectFloat(A=@n7, B=@n8, bPickA=@n4), bPickA=Get StartedTransitionToSecondPeekState())
   ) -> (then) {is=title K2Node_CallFunction}
   > Return Node(IsSizeZero=false)
  ]}
 GetContentSize {nodes [
   GetContentSize() {is=title K2Node_FunctionEntry}
   n1 = GetPaintSpaceGeometry!(self:Widget Object Reference=Get Overlay_BladeContent()) -> (ReturnValue:"Geometry Structure (by ref)")
   n2 = GetPaintSpaceGeometry(self=Get Grid_BladeBase() -> (Grid_BladeBase:Grid Panel Object Reference))
   > Return Node!(
    execute
    ReturnValue:Vector 2D Structure=Select(
     Option 0:Vector 2D Structure=GetLocalSize!(Geometry:"Geometry Structure (by ref)"=@n1) -> (ReturnValue:Vector 2D Structure)
     Option 1:Vector 2D Structure=GetLocalSize(Geometry=@n2)
     Index=Get UseFullscreenBlade()
    ) -> (ReturnValue:Vector 2D Structure)
   )
  ]}
 OnNotifyBladeResized {nodes [OnNotifyBladeResized() {is=title K2Node_FunctionEntry}]}
 OnNotifyParentBladeCatcherClicked {nodes [OnNotifyParentBladeCatcherClicked() {is=title K2Node_FunctionEntry}]}
 OnNotifyParentBladeContractStart {nodes [OnNotifyParentBladeContractStart() {is=title K2Node_FunctionEntry}]}
 OnNotifyParentBladeExpandStart {nodes [OnNotifyParentBladeExpandStart() {is=title K2Node_FunctionEntry}]}
 OnNotifyParentBladePeekStateTransitionFinished {nodes [OnNotifyParentBladePeekStateTransitionFinished() {is=title K2Node_FunctionEntry}]}
 OnNotifyParentBladeTransitionFinished {nodes [OnNotifyParentBladeTransitionFinished() {is=title K2Node_FunctionEntry}]}
 OnNotifyParentPeekButtonClicked {nodes [OnNotifyParentPeekButtonClicked() {is=title K2Node_FunctionEntry}]}
 SetCompactUIStyle {nodes [
   SetCompactUIStyle() {is=title K2Node_FunctionEntry}
   > Branch(Condition=IsUsingCompactUI())
   > SetSidesToPad!(
    execute
    self:Safe Zone Object Reference=Get SafeZone()
    InPadLeft:bool=true
    InPadRight:bool=true
    InPadTop:bool=false
    InPadBottom:bool=false
   ) -> (then)
  ]}
 SetHasClickCatcherButton {nodes [
   n1 = SetHasClickCatcherButton!() -> (then, InHasClickCatcherButton:bool=~) {is=title K2Node_FunctionEntry}
   > n2 = Set bHasClickCatcherButton(bHasClickCatcherButton:bool=@n1) -> (Output_Get:bool)
   n3 = "Not Equal (Enum)"!(A:E_BladeAnimationState Enum=Get EBladeAnimState() -> (EBladeAnimState=NewEnumerator0), B:E_BladeAnimationState Enum=NewEnumerator0) -> (ReturnValue:bool)
   > SetVisibility(
    self=Get EmptyButton_ClickCatcherCloseBlade()
    InVisibility=Select(Option 0:ESlateVisibility Enum=Collapsed, Option 1:ESlateVisibility Enum=Visible, Index=AND Boolean(A=@n2, B=@n3)) -> (ReturnValue:ESlateVisibility Enum)
   )
  ]}
 SetLerpAnimatorBetweenPeekStates {nodes [
   n1 = SetLerpAnimatorBetweenPeekStates!() -> (then, CurrentAnimatorValue:double=~) {is=title K2Node_FunctionEntry}
   > Set FinishedTransitionToSecondPeekState(FinishedTransitionToSecondPeekState:bool=Select(Option 0:bool=false, Option 1:bool=true, Index="Equal (Float)"(A=@n1, B=1)) -> (ReturnValue:bool)) -> (Output_Get:bool)
   > Branch(Condition="Not Equal (Enum)"(A=Get EBladeAnimState() -> (EBladeAnimState=NewEnumerator0), B=NewEnumerator4))
   n2 = IsUsingCompactUI()
   n3 = Get PeekPushPercentSecondary_CompactUI()
   n4 = Get PeekPushPercentSecondary()
   n5 = Get PeekPushPercent_CompactUI()
   n6 = Get PeekPushPercent()
   > CalculateNudgeValue(PeekValue=Lerp(A=SelectFloat(A=@n5, B=@n6, bPickA=@n2), B=SelectFloat(A=@n3, B=@n4, bPickA=@n2), Alpha=@n1))
   > n7 = Switch on E_BladeUIAnchor(Selection=Get EBladeAlignment() -> (EBladeAlignment=NewEnumerator0))
   n8 = Get NudgeForOffscreen()
   SetNudge(execute=[@n7.NewEnumerator1, @n7.NewEnumerator3], self=SlotAsGridSlot(Widget=Get Overlay_BladeContent()), InNudge=MakeVector2D(X=0, Y=@n8))
   SetNudge(execute=[@n7.NewEnumerator0, @n7.NewEnumerator2], self=SlotAsGridSlot(Widget=Get Overlay_BladeContent()), InNudge=MakeVector2D(X=@n8, Y=0))
  ]}
 SetLerpAnimatorExpandContractBlade {nodes [
   n1 = SetLerpAnimatorExpandContractBlade!() -> (then, CurrentAnimatorValue:double=~) {is=title K2Node_FunctionEntry}
   > n2 = Branch(Condition="Equal (Float)"(A=@n1))
   > n3 = Set EBladeAnimState(EBladeAnimState=NewEnumerator4) -> (Output_Get=NewEnumerator0)
   Set PreviousLerpAnimValue(execute=[@n17.then, @n3.then, @n13.then, @n16.then, @n14.then], PreviousLerpAnimValue:double=@n1) -> (Output_Get:double)
   > n4 = Switch on E_BladeUIAnchor(Selection=Get EBladeAlignment() -> (EBladeAlignment=NewEnumerator0))
   n5 = Lerp(A=0, B=Get NudgeForOffscreen(), Alpha=@n1)
   n6 = SetNudge(execute=[@n4.NewEnumerator1, @n4.NewEnumerator3], self=SlotAsGridSlot(Widget=Get Overlay_BladeContent()), InNudge=MakeVector2D(X=0, Y=@n5))
   n7 = "Not Equal (Enum)"(A=Get EBladeAnimState() -> (EBladeAnimState=NewEnumerator0), B=NewEnumerator0)
   n8 = Get bHasClickCatcherButton()
   SetVisibility(
    execute=[@n6.then, @n11.then]
    self=Get EmptyButton_ClickCatcherCloseBlade()
    InVisibility=Select(Option 0:ESlateVisibility Enum=Collapsed, Option 1:ESlateVisibility Enum=Visible, Index=AND Boolean(A=@n7, B=@n8)) -> (ReturnValue:ESlateVisibility Enum)
   )
   n9 = NOT Boolean(A=Get HidePeekButton())
   n10 = "Equal (Enum)"(A=Get EBladeAnimState() -> (EBladeAnimState=NewEnumerator0), B=NewEnumerator0)
   > SetVisibility(self=Get EmptyButton_ForPeekMode(), InVisibility=Select(Option 0=Collapsed, Option 1=Visible, Index=AND Boolean(A=@n10, B=@n9)))
   n11 = SetNudge(execute=[@n4.NewEnumerator0, @n4.NewEnumerator2], self=SlotAsGridSlot(Widget=Get Overlay_BladeContent()), InNudge=MakeVector2D(X=@n5, Y=0))
   n12 = Branch(execute=@n2.else, Condition="Equal (Float)"(A=@n1, B=1))
   > n13 = Set EBladeAnimState(EBladeAnimState=NewEnumerator0) -> (Output_Get=NewEnumerator0)
   n14 = Branch(execute=@n12.else, Condition="Equal (Float)"(A=Get PreviousLerpAnimValue() -> (PreviousLerpAnimValue:double), B=@n1))
   n15 = Branch(execute=@n14.else, Condition="float < float"(A=Get PreviousLerpAnimValue(), B=@n1))
   > n16 = Set EBladeAnimState(EBladeAnimState=NewEnumerator1) -> (Output_Get=NewEnumerator0)
   n17 = Set EBladeAnimState(execute=@n15.else, EBladeAnimState=NewEnumerator2) -> (Output_Get=NewEnumerator0)
  ]}
 SetPeekButtonVisibility {nodes [
   n1 = SetPeekButtonVisibility!() -> (then, HidePeekButton:bool=~) {is=title K2Node_FunctionEntry}
   > n2 = Set HidePeekButton(HidePeekButton:bool=@n1) -> (Output_Get:bool)
   > SetVisibility(self=Get EmptyButton_ForPeekMode(), InVisibility=Select(Option 0=Visible, Option 1=Collapsed, Index=@n2))
  ]}
 SetPeekPushPercentFromWidgetSize {nodes [
   SetPeekPushPercentFromWidgetSize!() -> (then, WidgetSize:double=~, PercentageOfWidgetToShow:double=~, AditionnalOffset:double=~, PercentageOfWidgetToShowSecondary:double=1.1) {is=title K2Node_FunctionEntry}
   n1 = GetContentSize!() -> (ReturnValue_X:double, ReturnValue_Y:double) {is=title K2Node_CallFunction}
   n2 = Get EBladeAlignment() -> (EBladeAlignment=NewEnumerator0)
   n3 = "Equal (Enum)"(A:E_BladeUIAnchor Enum=@n2, B:E_BladeUIAnchor Enum=NewEnumerator0)
   n4 = "Equal (Enum)"(A=@n2, B=NewEnumerator2)
   > Set ContentSize(ContentSize:double=SelectFloat(A=@n1, B=@n1.ReturnValue_Y, bPickA=OR Boolean(A=@n3, B=@n4))) -> (Output_Get:double)
   n5 = float * float(A=Get WidgetSize() -> (WidgetSize:double), B=Get PercentageOfWidgetToShow() -> (PercentageOfWidgetToShow:double))
   n6 = Get ContentSize() -> (ContentSize:double)
   > Set NudgeValue(NudgeValue:double=float + float(A=float - float(A=@n6, B=@n5), B=Get AditionnalOffset() -> (AditionnalOffset:double))) -> (Output_Get:double)
   n7 = float * float(A=Get PercentageOfWidgetToShowSecondary() -> (PercentageOfWidgetToShowSecondary:double), B=Get WidgetSize())
   n8 = Get ContentSize()
   > Set SecondaryNudgeValue(SecondaryNudgeValue:double=float + float(A=float - float(A=@n8, B=@n7), B=Get AditionnalOffset())) -> (Output_Get:double)
   > n9 = Branch(Condition=IsUsingCompactUI())
   n10 = SafeDivide(A=Get NudgeValue() -> (NudgeValue:double), B=Get ContentSize())
   > Set PeekPushPercent_CompactUI(PeekPushPercent_CompactUI:double=@n10) -> (Output_Get:double)
   n11 = SafeDivide(A=Get SecondaryNudgeValue() -> (SecondaryNudgeValue:double), B=Get ContentSize())
   > n12 = Set PeekPushPercentSecondary_CompactUI(PeekPushPercentSecondary_CompactUI:double=@n11) -> (Output_Get:double)
   Refresh Nudge Position(execute=[@n12.then, @n13.then]) {is=title K2Node_CallFunction}
   Set PeekPushPercent(execute=@n9.else, PeekPushPercent:double=@n10) -> (Output_Get:double)
   > n13 = Set PeekPushPercentSecondary(PeekPushPercentSecondary:double=@n11) -> (Output_Get:double)
  ]}
 SetPeekPushPercentages {nodes [
   n1 = SetPeekPushPercentages!() -> (then, PeekPushPercent:double=~, PeekPrushPercentSecondary:double=~) {is=title K2Node_FunctionEntry}
   > Set PeekPushPercent(PeekPushPercent=@n1)
   > Set PeekPushPercentSecondary(PeekPushPercentSecondary=@n1.PeekPrushPercentSecondary)
  ]}
 SetShouldStartInSecondPeekState {nodes [
   n1 = SetShouldStartInSecondPeekState!() -> (then, ShouldStartInSecondPeekState:bool=~) {is=title K2Node_FunctionEntry}
   > n2 = Set StartedTransitionToSecondPeekState(StartedTransitionToSecondPeekState=@n1)
   > Set FinishedTransitionToSecondPeekState(FinishedTransitionToSecondPeekState=@n2)
  ]}
]
variables [
 AnimateBackplateWhenExpanding {type=boolean}
 BackplateImageBrush {
  type=SlateBrush
  default {
   bIsDynamicallyLoaded=false
   drawAs=Image
   imageType=NoImage
   mirroring=NoMirror
   resourceName=None
   resourceObject=/DiscoveryBrowser/Blade/Materials/M_UI_RadialGradientBladeBackground
   tiling=NoTile
   imageSize=<32 32>
   margin=<ltrb 0 0 0 0>
   outlineSettings {
    bUseBrushTransparency=false
    roundingType=HalfHeightRadius
    width=0
    color {colorUseRule=UseColor_Specified, specifiedColor=<rgba 0 0 0 0>}
    cornerRadii=<0 0 0 1>
   }
   tintColor {colorUseRule=UseColor_Specified, specifiedColor=<rgba 1 1 1 1>}
   uVRegion {bIsValid=false, max=<0 0>, min=<0 0>}
  }
 }
 bHasClickCatcherButton {type=boolean, default=true}
 bIsClickCatcherButtonTransparent {type=boolean}
 bStartExpanded {type=boolean, default=true}
 ClickCatcherButtonStyle {
  type=ButtonStyle
  default {
   clickedSlateSound {resourceObject=null}
   disabled {
    bIsDynamicallyLoaded=false
    drawAs=Image
    imageType=NoImage
    mirroring=NoMirror
    resourceName=None
    resourceObject=null
    tiling=NoTile
    imageSize=<32 32>
    margin=<ltrb 0 0 0 0>
    outlineSettings {
     bUseBrushTransparency=false
     roundingType=HalfHeightRadius
     width=0
     color {colorUseRule=UseColor_Specified, specifiedColor=<rgba 0 0 0 0>}
     cornerRadii=<0 0 0 1>
    }
    tintColor {colorUseRule=UseColor_Specified, specifiedColor=<rgba 0 0.0044 0.0319 0.5>}
    uVRegion {bIsValid=false, max=<0 0>, min=<0 0>}
   }
   disabledForeground {colorUseRule=UseColor_Foreground, specifiedColor=<rgba 1 0 1 1>}
   hovered {
    bIsDynamicallyLoaded=false
    drawAs=Image
    imageType=NoImage
    mirroring=NoMirror
    resourceName=None
    resourceObject=null
    tiling=NoTile
    imageSize=<32 32>
    margin=<ltrb 0 0 0 0>
    outlineSettings {
     bUseBrushTransparency=false
     roundingType=HalfHeightRadius
     width=0
     color {colorUseRule=UseColor_Specified, specifiedColor=<rgba 0 0 0 0>}
     cornerRadii=<0 0 0 1>
    }
    tintColor {colorUseRule=UseColor_Specified, specifiedColor=<rgba 0 0.0044 0.0319 0.5>}
    uVRegion {bIsValid=false, max=<0 0>, min=<0 0>}
   }
   hoveredForeground {colorUseRule=UseColor_Foreground, specifiedColor=<rgba 1 0 1 1>}
   hoveredSlateSound {resourceObject=null}
   normal {
    bIsDynamicallyLoaded=false
    drawAs=Image
    imageType=NoImage
    mirroring=NoMirror
    resourceName=None
    resourceObject=null
    tiling=NoTile
    imageSize=<32 32>
    margin=<ltrb 0 0 0 0>
    outlineSettings {
     bUseBrushTransparency=false
     roundingType=HalfHeightRadius
     width=0
     color {colorUseRule=UseColor_Specified, specifiedColor=<rgba 0 0 0 0>}
     cornerRadii=<0 0 0 1>
    }
    tintColor {colorUseRule=UseColor_Specified, specifiedColor=<rgba 0 0.0044 0.0319 0.5>}
    uVRegion {bIsValid=false, max=<0 0>, min=<0 0>}
   }
   normalForeground {colorUseRule=UseColor_Foreground, specifiedColor=<rgba 1 0 1 1>}
   normalPadding=<ltrb 0 0 0 0>
   pressed {
    bIsDynamicallyLoaded=false
    drawAs=Image
    imageType=NoImage
    mirroring=NoMirror
    resourceName=None
    resourceObject=null
    tiling=NoTile
    imageSize=<32 32>
    margin=<ltrb 0 0 0 0>
    outlineSettings {
     bUseBrushTransparency=false
     roundingType=HalfHeightRadius
     width=0
     color {colorUseRule=UseColor_Specified, specifiedColor=<rgba 0 0 0 0>}
     cornerRadii=<0 0 0 1>
    }
    tintColor {colorUseRule=UseColor_Specified, specifiedColor=<rgba 0 0.0044 0.0319 0.5>}
    uVRegion {bIsValid=false, max=<0 0>, min=<0 0>}
   }
   pressedForeground {colorUseRule=UseColor_Foreground, specifiedColor=<rgba 1 0 1 1>}
   pressedPadding=<ltrb 0 0 0 0>
   pressedSlateSound {resourceObject=/Game/Sounds/Fort_UI_Sounds/FrontEnd/New_FrontEnd/Fortnite_UI_MainMenu_Generic_01_Cue}
  }
 }
 Column_Content {type=integer}
 ColumnSpan_Content {type=integer}
 ContentHeight {type=number}
 ContentPadding {type=Margin, default=<ltrb 0 0 0 0>}
 ContentWidth {type=number}
 EBladeAlignment {type=E_BladeUIAnchor, default=Left}
 EBladeAnimState {type=E_BladeAnimationState, default=Contracted}
 EnableSecondaryPeekState {type=boolean}
 FinishedTransitionToSecondPeekState {type=boolean}
 HidePeekButton {type=boolean, default=true}
 LerpAnimatorBetweenPeekStates {type=number}
 LerpAnimatorExpandContractBlade {type=number}
 NudgeForOffscreen {type=number}
 OnNotifyBladeResized {}
 OnNotifyParentBladeCatcherClicked {}
 OnNotifyParentBladeContractStart {}
 OnNotifyParentBladeExpandStart {}
 OnNotifyParentBladePeekStateTransitionFinished {}
 OnNotifyParentBladeTransitionFinished {}
 OnNotifyParentPeekButtonClicked {}
 PeekPushPercent {type=number}
 PeekPushPercent_CompactUI {type=number}
 PeekPushPercentSecondary {type=number}
 PeekPushPercentSecondary_CompactUI {type=number}
 PreviousLerpAnimValue {type=number}
 Row_Content {type=integer}
 RowSpan_Content {type=integer}
 StartedTransitionToSecondPeekState {type=boolean}
 TransparentButtonStyle {
  type=ButtonStyle
  default {
   clickedSlateSound {resourceObject=null}
   disabled {
    bIsDynamicallyLoaded=false
    drawAs=NoDrawType
    imageType=NoImage
    mirroring=NoMirror
    resourceName=None
    resourceObject=null
    tiling=NoTile
    imageSize=<0 0>
    margin=<ltrb 0 0 0 0>
    outlineSettings {
     bUseBrushTransparency=false
     roundingType=HalfHeightRadius
     width=0
     color {colorUseRule=UseColor_Specified, specifiedColor=<rgba 0 0 0 0>}
     cornerRadii=<0 0 0 1>
    }
    tintColor {colorUseRule=UseColor_Specified, specifiedColor=<rgba 1 1 1 0>}
    uVRegion {bIsValid=false, max=<0 0>, min=<0 0>}
   }
   disabledForeground {colorUseRule=UseColor_Foreground, specifiedColor=<rgba 1 0 1 1>}
   hovered {
    bIsDynamicallyLoaded=false
    drawAs=NoDrawType
    imageType=NoImage
    mirroring=NoMirror
    resourceName=None
    resourceObject=null
    tiling=NoTile
    imageSize=<32 32>
    margin=<ltrb 0 0 0 0>
    outlineSettings {
     bUseBrushTransparency=false
     roundingType=HalfHeightRadius
     width=0
     color {colorUseRule=UseColor_Specified, specifiedColor=<rgba 0 0 0 0>}
     cornerRadii=<0 0 0 1>
    }
    tintColor {colorUseRule=UseColor_Specified, specifiedColor=<rgba 1 1 1 0>}
    uVRegion {bIsValid=false, max=<0 0>, min=<0 0>}
   }
   hoveredForeground {colorUseRule=UseColor_Foreground, specifiedColor=<rgba 1 0 1 1>}
   hoveredSlateSound {resourceObject=null}
   normal {
    bIsDynamicallyLoaded=false
    drawAs=NoDrawType
    imageType=NoImage
    mirroring=NoMirror
    resourceName=None
    resourceObject=null
    tiling=NoTile
    imageSize=<32 32>
    margin=<ltrb 0 0 0 0>
    outlineSettings {
     bUseBrushTransparency=false
     roundingType=HalfHeightRadius
     width=0
     color {colorUseRule=UseColor_Specified, specifiedColor=<rgba 0 0 0 0>}
     cornerRadii=<0 0 0 1>
    }
    tintColor {colorUseRule=UseColor_Specified, specifiedColor=<rgba 1 1 1 0>}
    uVRegion {bIsValid=false, max=<0 0>, min=<0 0>}
   }
   normalForeground {colorUseRule=UseColor_Foreground, specifiedColor=<rgba 1 0 1 1>}
   normalPadding=<ltrb 0 0 0 0>
   pressed {
    bIsDynamicallyLoaded=false
    drawAs=NoDrawType
    imageType=NoImage
    mirroring=NoMirror
    resourceName=None
    resourceObject=null
    tiling=NoTile
    imageSize=<32 32>
    margin=<ltrb 0 0 0 0>
    outlineSettings {
     bUseBrushTransparency=false
     roundingType=HalfHeightRadius
     width=0
     color {colorUseRule=UseColor_Specified, specifiedColor=<rgba 0 0 0 0>}
     cornerRadii=<0 0 0 1>
    }
    tintColor {colorUseRule=UseColor_Specified, specifiedColor=<rgba 1 1 1 0>}
    uVRegion {bIsValid=false, max=<0 0>, min=<0 0>}
   }
   pressedForeground {colorUseRule=UseColor_Foreground, specifiedColor=<rgba 1 0 1 1>}
   pressedPadding=<ltrb 0 0 0 0>
   pressedSlateSound {resourceObject=null}
  }
 }
 UseFullscreenBlade {type=boolean}
]
widgets {
 info {inheritedWidgetCount=0, parentClass=/Script/FortniteUI.FortBladeWidget, rootWidgetClass=&GridPanel, widgetCount=8}
 widgets [
  Grid_BladeBase {widgetClassPath=&GridPanel, bIsVariable=true, widget=Grid_BladeBase, properties {columnFill [0,1,0], rowFill [0,1,0]}}
  EmptyButton_ClickCatcherCloseBlade {
   widgetClassPath=&Button
   parent=Grid_BladeBase
   widget=EmptyButton_ClickCatcherCloseBlade
   properties {
    isFocusable=false
    visibility=Collapsed
    widgetStyle {
     clickedSlateSound {resourceObject=null}
     disabled {
      bIsDynamicallyLoaded=false
      drawAs=Image
      imageType=NoImage
      mirroring=NoMirror
      resourceName=None
      resourceObject=null
      tiling=NoTile
      imageSize=<32 32>
      margin=<ltrb 0 0 0 0>
      outlineSettings {
       bUseBrushTransparency=false
       roundingType=HalfHeightRadius
       width=0
       color {colorUseRule=UseColor_Specified, specifiedColor=<rgba 0 0 0 0>}
       cornerRadii=<0 0 0 0>
      }
      tintColor {colorUseRule=UseColor_Specified, specifiedColor=<rgba 0 0.0044 0.0319 0.5>}
      uVRegion {bIsValid=false, max=<0 0>, min=<0 0>}
     }
     disabledForeground {colorUseRule=UseColor_Specified, specifiedColor=<rgba 0.5271 0.5271 0.5271 1>}
     hovered {
      bIsDynamicallyLoaded=false
      drawAs=Image
      imageType=NoImage
      mirroring=NoMirror
      resourceName=None
      resourceObject=null
      tiling=NoTile
      imageSize=<32 32>
      margin=<ltrb 0 0 0 0>
      outlineSettings {
       bUseBrushTransparency=false
       roundingType=HalfHeightRadius
       width=0
       color {colorUseRule=UseColor_Specified, specifiedColor=<rgba 0 0 0 0>}
       cornerRadii=<0 0 0 1>
      }
      tintColor {colorUseRule=UseColor_Specified, specifiedColor=<rgba 0 0.0044 0.0319 0.5>}
      uVRegion {bIsValid=false, max=<0 0>, min=<0 0>}
     }
     hoveredForeground {colorUseRule=UseColor_Specified, specifiedColor=<rgba 0 1 1 1>}
     hoveredSlateSound {resourceObject=null}
     normal {
      bIsDynamicallyLoaded=false
      drawAs=Image
      imageType=NoImage
      mirroring=NoMirror
      resourceName=None
      resourceObject=null
      tiling=NoTile
      imageSize=<32 32>
      margin=<ltrb 0 0 0 0>
      outlineSettings {
       bUseBrushTransparency=false
       roundingType=HalfHeightRadius
       width=0
       color {colorUseRule=UseColor_Specified, specifiedColor=<rgba 0 0 0 0>}
       cornerRadii=<0 0 0 1>
      }
      tintColor {colorUseRule=UseColor_Specified, specifiedColor=<rgba 0 0.0044 0.0319 0.5>}
      uVRegion {bIsValid=false, max=<0 0>, min=<0 0>}
     }
     normalForeground {colorUseRule=UseColor_Specified, specifiedColor=<rgba 0 1 1 1>}
     normalPadding=<ltrb 12 1.5 12 1.5>
     pressed {
      bIsDynamicallyLoaded=false
      drawAs=Image
      imageType=NoImage
      mirroring=NoMirror
      resourceName=None
      resourceObject=null
      tiling=NoTile
      imageSize=<32 32>
      margin=<ltrb 0 0 0 0>
      outlineSettings {
       bUseBrushTransparency=false
       roundingType=HalfHeightRadius
       width=0
       color {colorUseRule=UseColor_Specified, specifiedColor=<rgba 0 0 0 0>}
       cornerRadii=<0 0 0 1>
      }
      tintColor {colorUseRule=UseColor_Specified, specifiedColor=<rgba 0 0.0044 0.0319 0.5>}
      uVRegion {bIsValid=false, max=<0 0>, min=<0 0>}
     }
     pressedForeground {colorUseRule=UseColor_Specified, specifiedColor=<rgba 0 1 1 1>}
     pressedPadding=<ltrb 12 1.5 12 1.5>
     pressedSlateSound {resourceObject=null}
    }
   }
   slot {class=GridSlot, properties {columnSpan=3, rowSpan=3}}
  }
  Overlay_BladeContent {
   widgetClassPath=&Overlay
   bIsVariable=true
   parent=Grid_BladeBase
   widget=Overlay_BladeContent
   slot {class=GridSlot, properties {column=2, columnSpan=1, rowSpan=3}}
  }
  FullBackplate_Image {
   widgetClassPath=&Image
   bIsVariable=true
   parent=Overlay_BladeContent
   widget=FullBackplate_Image
   slot {class=OverlaySlot, properties {horizontalAlignment=HAlign_Fill, verticalAlignment=VAlign_Fill}}
  }
  SafeZone {
   widgetClassPath=&SafeZone
   bIsVariable=true
   parent=Overlay_BladeContent
   widget=SafeZone
   properties {padBottom=false, padLeft=false, padRight=false, padTop=false}
   slot {class=OverlaySlot, properties {horizontalAlignment=HAlign_Fill, verticalAlignment=VAlign_Fill}}
  }
  NamedSlot_BladeContent {widgetClassPath=&NamedSlot, bIsVariable=true, parent=SafeZone, widget=NamedSlot_BladeContent, slot {class=SafeZoneSlot}}
  EmptyButton_ForPeekMode {
   widgetClassPath=&Button
   parent=Overlay_BladeContent
   widget=EmptyButton_ForPeekMode
   properties {isFocusable=false, visibility=Collapsed}
   slot {class=OverlaySlot, properties {horizontalAlignment=HAlign_Fill, verticalAlignment=VAlign_Fill}}
  }
  NamedSlot_BladeContent_Foreground {
   widgetClassPath=&NamedSlot
   bIsVariable=true
   parent=Grid_BladeBase
   widget=NamedSlot_BladeContent_Foreground
   slot {class=GridSlot, properties {columnSpan=3, layer=100, rowSpan=3}}
  }
 ]
}
