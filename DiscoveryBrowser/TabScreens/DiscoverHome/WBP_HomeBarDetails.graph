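# Details widget for the Discover home bar, showing the selected game or island's name, status and friends playing.
# A switcher holds game-selected and joinable-party layouts, both updated with a friend-count badge and singular/plural label.
# SetCurrentOrInvalidReasonText maps the invalid-activity reason or current-lobby flag to status text and lock icon visibility, always collapsing the selected-lobby icon.
# UpdateDownloadProgressText reuses that status text for a rounded download percentage.
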WidgetBlueprint /DiscoveryBrowser/TabScreens/DiscoverHome/WBP_HomeBarDetails
graphs [
 EventGraph {nodes [
   Event PreConstruct()
   > "SetText (Text)"(self=Get Text_Subtext() -> (Text_Subtext:Common Text Object Reference), InText="LOCTEXT(\"Recently Played && Favorites\")")
   Event Construct()
   Event Tick()
  ]}
 SetCurrentOrInvalidReasonText {nodes [
   n1 = SetCurrentOrInvalidReasonText!() -> (then, InvalidActivityReason:EFortInvalidActivityReason Enum=None, IsCurrentActivity:bool=~) {is=title K2Node_FunctionEntry}
   n2 = "Equal (Enum)"(A:EFortInvalidActivityReason Enum=@n1, B:EFortInvalidActivityReason Enum=None)
   > SetVisibility(
    self=Get StatusIndicator_Locked() -> (StatusIndicator_Locked:WBP UIKit Status Indicator Object Reference)
    InVisibility=Select(Option 0=HitTestInvisible, Option 1=Collapsed, Index=OR Boolean(A=@n2, B=@n1.IsCurrentActivity))
   )
   n3 = Select(
    Option 0:Text=Select!(
     None:Text
     PartyTooBig:Text="LOCTABLE(\"/Game/UI/DiscoveryShared/StringTable_ActivityBrowser\", \"Warning_PartyTooLarge\")"
     NotPartyLeader:Text="LOCTEXT(\"Only party leaders can change games.\")"
     MatchmakingAlready:Text="LOCTABLE(\"&StringTable_ActivityBrowser\", \"Warning_CurrentlyMatchmaking\")"
     PlaylistDisabled:Text="LOCTABLE(\"&StringTable_ActivityBrowser\", \"Warning_ModeDisabled\")"
     PlayerIsLfg:Text="LOCTABLE(\"&StringTable_ActivityBrowser\", \"Warning_PlayerIsLfg\")"
     InvalidData:Text="LOCTABLE(\"&StringTable_ActivityBrowser\", \"Warning_InvalidMode\")"
     ExpiredTournamentLockCache:Text
     Index:EFortInvalidActivityReason Enum=@n1
    ) -> (ReturnValue)
    Option 1:Text="LOCTEXT(\"Current Lobby\")"
    Index=@n1.IsCurrentActivity
   ) -> (ReturnValue:Text)
   > "SetText (Text)"(self=Get Text_GameStatus() -> (Text_GameStatus:Common Text Object Reference), InText=@n3)
   > SetVisibility(
    self=Get Text_GameStatus()
    InVisibility=Select!(Option 0:ESlateVisibility Enum=SelfHitTestInvisible, Option 1:ESlateVisibility Enum=Collapsed, Index:bool=TextIsEmpty(InText=@n3)) -> (ReturnValue:ESlateVisibility Enum)
   )
   > SetVisibility(self=Get StatusIndicator_SelectedLobby() -> (StatusIndicator_SelectedLobby:Image Object Reference), InVisibility=Collapsed)
   n4 = TextIsEmpty(InText=Get Text(self:Text Object Reference=Get Text_GameStatus()))
   n5 = NOT Boolean(A=Get IsCurrentActivity() -> (IsCurrentActivity:bool))
   > SetVisibility(
    self=Get HB_ActivityInfo() -> (HB_ActivityInfo:Horizontal Box Object Reference)
    InVisibility=Select(Option 0=HitTestInvisible, Option 1=Collapsed, Index=AND Boolean(A=@n4, B=@n5))
   )
   Select(Option 0=Collapsed, Option 1=HitTestInvisible, Index=@n1.IsCurrentActivity)
  ]}
 SetFriendsCountText {nodes [
   n1 = SetFriendsCountText!() -> (then, FriendsCount:int) {is=title K2Node_FunctionEntry}
   > SetBadgeText!(
    execute
    self:WBP UIKit Badge Base Object Reference=[
     Get WBP_UIKit_Badge() -> (WBP_UIKit_Badge:WBP UIKit Badge Object Reference)
     Get WBP_UIKit_Badge_69() -> (WBP_UIKit_Badge_69:WBP UIKit Badge Object Reference)
    ]
    Text:Text="To Text (Integer)"(Value=@n1)
   ) -> (then)
   > SetVisibility(
    self=[
     Get HB_GameIsland_FriendCount() -> (HB_GameIsland_FriendCount:Horizontal Box Object Reference)
     Get HB_JP_FriendCount() -> (HB_JP_FriendCount:Horizontal Box Object Reference)
    ]
    InVisibility=Select(Option 0=Collapsed, Option 1=HitTestInvisible, Index="integer > integer"(A=@n1))
   )
   > "SetText (Text)"(
    self=[Get Text_Playing_1() -> (Text_Playing_1:Common Text Object Reference), Get Text_Playing_2() -> (Text_Playing_2:Common Text Object Reference)]
    InText=Select(Option 0:Text="LOCTEXT(\"FRIEND PLAYING\")", Option 1:Text="LOCTEXT(\"FRIENDS PLAYING\")", Index="integer > integer"(A=@n1, B=1)) -> (ReturnValue:Text)
   )
  ]}
 UpdateDownloadProgressText {nodes [
   n1 = UpdateDownloadProgressText!() -> (then, DownloadProgress:double=~) {is=title K2Node_FunctionEntry}
   n2 = Round(A=float * float(A=@n1, B=100) {is=title K2Node_PromotableOperator}) {is=title K2Node_CallFunction}
   > "SetText (Text)"(
    self=Get Text_GameStatus()
    InText=Format Text!(Format="LOCTEXT(\"Downloading {Percent}%\")", Percent:double="To Float (Integer)"(InInt=@n2)) -> (Result)
   )
  ]}
]
variables []
widgets {
 info {inheritedWidgetCount=0, parentClass=&UserWidget, rootWidgetClass=&StackBox, widgetCount=17}
 widgets [
  StackBox_0 {widgetClassPath=&StackBox, widget=StackBox_0, properties {orientation=Orient_Vertical}}
  Text_SelectedTile {
   widgetClassPath=&CommonTextBlock
   bIsVariable=true
   parent=StackBox_0
   widget=Text_SelectedTile
   properties {
    lineHeightPercentage=0.8
    text="{Game / Island Name}"
    textTransformPolicy=ToUpper
    wrapTextAt=840
    font {
     bForceMonospaced=false
     bMaterialIsStencil=false
     fontMaterial=null
     fontObject=&HeadingNow
     letterSpacing=-40
     monospacedWidth=1
     size=32.25
     skewAmount=0
     typefaceFontName=86 Bold Italic
     outlineSettings {
      bApplyOutlineToDropShadows=false
      bMiteredCorners=false
      bSeparateFillAlpha=false
      outlineMaterial=/FortUILibrary/MaterialInstances/Text/MI_UI_Block_Text_Outline
      outlineSize=2
      outlineColor=<rgba 0 0 0 1>
     }
    }
    shadowOffset=<0 0>
   }
   slot {class=StackBoxSlot}
  }
  Text_Subtext {
   widgetClassPath=&CommonTextBlock
   bIsVariable=true
   parent=StackBox_0
   widget=Text_Subtext
   properties {
    lineHeightPercentage=0.9
    text="{Subtext}"
    visibility=Collapsed
    font {
     bForceMonospaced=false
     bMaterialIsStencil=false
     fontMaterial=null
     fontObject=&HeadingNow
     letterSpacing=0
     monospacedWidth=1
     size=20.25
     skewAmount=0
     typefaceFontName=74 Regular
     outlineSettings {
      bApplyOutlineToDropShadows=false
      bMiteredCorners=false
      bSeparateFillAlpha=false
      outlineMaterial=&MI_UI_Block_Text_Outline
      outlineSize=1
      outlineColor=<rgba 0 0 0 1>
     }
    }
    shadowOffset=<0 0>
   }
   slot {class=StackBoxSlot, properties {padding=<ltrb 0 8 0 0>}}
  }
  Switcher_GameDetailsContent {
   widgetClassPath=&CommonActivatableWidgetSwitcher
   bIsVariable=true
   parent=StackBox_0
   widget=Switcher_GameDetailsContent
   slot {class=StackBoxSlot}
  }
  VB_GameSelected {
   widgetClassPath=&VerticalBox
   parent=Switcher_GameDetailsContent
   widget=VB_GameSelected
   slot {class=WidgetSwitcherSlot, properties {padding=<ltrb 0 8 0 0>}}
  }
  HB_GameIsland_FriendCount {
   widgetClassPath=&HorizontalBox
   bIsVariable=true
   parent=VB_GameSelected
   widget=HB_GameIsland_FriendCount
   slot {class=VerticalBoxSlot, properties {horizontalAlignment=HAlign_Left, verticalAlignment=VAlign_Top, padding=<ltrb 0 0 24 4>}}
  }
  WBP_UIKit_Badge {
   widgetClassPath=/FortUILibrary/UIKit/Badge/WBP_UIKit_Badge
   bIsVariable=true
   parent=HB_GameIsland_FriendCount
   widget=WBP_UIKit_Badge
   properties {badgeText=3}
   slot {class=HorizontalBoxSlot, properties {verticalAlignment=VAlign_Center}}
  }
  Text_Playing_1 {
   widgetClassPath=&CommonTextBlock
   bIsVariable=true
   parent=HB_GameIsland_FriendCount
   widget=Text_Playing_1
   properties {
    text=PLAYING
    colorAndOpacity {colorUseRule=UseColor_Specified, specifiedColor=<rgba 0.5776 0.5776 0.5776 1>}
    font {
     bForceMonospaced=false
     bMaterialIsStencil=false
     fontMaterial=null
     fontObject=&HeadingNow
     letterSpacing=0
     monospacedWidth=1
     size=15.75
     skewAmount=0
     typefaceFontName=86 Bold Italic
     outlineSettings {
      bApplyOutlineToDropShadows=false
      bMiteredCorners=false
      bSeparateFillAlpha=false
      outlineMaterial=&MI_UI_Block_Text_Outline
      outlineSize=2
      outlineColor=<rgba 0 0 0 1>
     }
    }
    shadowOffset=<0 0>
   }
   slot {class=HorizontalBoxSlot, properties {verticalAlignment=VAlign_Center, padding=<ltrb 8 0 0 0>}}
  }
  HB_ActivityInfo {
   widgetClassPath=&HorizontalBox
   bIsVariable=true
   parent=VB_GameSelected
   widget=HB_ActivityInfo
   slot {class=VerticalBoxSlot, properties {horizontalAlignment=HAlign_Left}}
  }
  StatusIndicator_Locked {
   widgetClassPath=/FortUILibrary/UIKit/StatusIndicator/WBP_UIKit_StatusIndicator
   bIsVariable=true
   parent=HB_ActivityInfo
   widget=StatusIndicator_Locked
   properties {type=Utility, icon=/FortUILibrary/Textures/MaterialInstances/MI_UIKit_Icon_Locked, overrideStatusIndicatorVisibility=true}
   slot {class=HorizontalBoxSlot, properties {horizontalAlignment=HAlign_Left, verticalAlignment=VAlign_Center, padding=<ltrb 0 0 4 0>}}
  }
  StatusIndicator_SelectedLobby {
   widgetClassPath=&Image
   bIsVariable=true
   parent=HB_ActivityInfo
   widget=StatusIndicator_SelectedLobby
   properties {brush {
     bIsDynamicallyLoaded=false
     drawAs=Image
     imageType=FullColor
     mirroring=NoMirror
     resourceName=None
     resourceObject=/DiscoveryBrowser/Tiles/Islands/Materials/MI_UI_IslandTile_State_SelectedGame
     tiling=NoTile
     imageSize=<32 32>
     margin=<ltrb 0 0 0 0>
     outlineSettings {
      bUseBrushTransparency=false
      roundingType=HalfHeightRadius
      width=0
      color {colorUseRule=UseColor_Specified, specifiedColor=<rgba 0 0 0 0>}
      cornerRadii=<0 0 0 0>
     }
     tintColor {colorUseRule=UseColor_Specified, specifiedColor=<rgba 1 1 1 1>}
     uVRegion {bIsValid=false, max=<0 0>, min=<0 0>}
    }}
   slot {class=HorizontalBoxSlot, properties {horizontalAlignment=HAlign_Center, verticalAlignment=VAlign_Center}}
  }
  Text_GameStatus {
   widgetClassPath=&CommonTextBlock
   bIsVariable=true
   parent=HB_ActivityInfo
   widget=Text_GameStatus
   properties {
    lineHeightPercentage=0.9
    text="{Current Lobby or Reason Disabled}"
    visibility=Collapsed
    font {
     bForceMonospaced=false
     bMaterialIsStencil=false
     fontMaterial=null
     fontObject=&HeadingNow
     letterSpacing=0
     monospacedWidth=1
     size=20.25
     skewAmount=0
     typefaceFontName=74 Regular
     outlineSettings {
      bApplyOutlineToDropShadows=false
      bMiteredCorners=false
      bSeparateFillAlpha=false
      outlineMaterial=&MI_UI_Block_Text_Outline
      outlineSize=1
      outlineColor=<rgba 0 0 0 1>
     }
    }
    shadowOffset=<0 0>
   }
   slot {class=HorizontalBoxSlot}
  }
  VB_JoinableParty {
   widgetClassPath=&VerticalBox
   parent=Switcher_GameDetailsContent
   widget=VB_JoinableParty
   slot {class=WidgetSwitcherSlot, properties {horizontalAlignment=HAlign_Left}}
  }
  JP_IslandName {
   widgetClassPath=/FortUILibrary/UIKit/Highlight/WBP_UIKit_Highlight
   bIsVariable=true
   parent=VB_JoinableParty
   widget=JP_IslandName
   properties {highlightSize=Compact, highlightText=Label, highlightType=Utility}
   slot {class=VerticalBoxSlot, properties {horizontalAlignment=HAlign_Left}}
  }
  HB_JP_FriendCount {
   widgetClassPath=&HorizontalBox
   bIsVariable=true
   parent=VB_JoinableParty
   widget=HB_JP_FriendCount
   slot {class=VerticalBoxSlot, properties {padding=<ltrb 0 8 0 0>}}
  }
  WBP_UIKit_Badge_69 {
   widgetClassPath=&WBP_UIKit_Badge
   bIsVariable=true
   parent=HB_JP_FriendCount
   widget=WBP_UIKit_Badge_69
   slot {class=HorizontalBoxSlot, properties {verticalAlignment=VAlign_Center}}
  }
  Text_Playing_2 {
   widgetClassPath=&CommonTextBlock
   bIsVariable=true
   parent=HB_JP_FriendCount
   widget=Text_Playing_2
   properties {
    text=PLAYING
    colorAndOpacity {colorUseRule=UseColor_Specified, specifiedColor=<rgba 0.5776 0.5776 0.5776 1>}
    font {
     bForceMonospaced=false
     bMaterialIsStencil=false
     fontMaterial=null
     fontObject=&BurbankSmall
     letterSpacing=0
     monospacedWidth=1
     size=13
     skewAmount=0
     typefaceFontName=Black
     outlineSettings {
      bApplyOutlineToDropShadows=false
      bMiteredCorners=false
      bSeparateFillAlpha=false
      outlineMaterial=&MI_UI_Block_Text_Outline
      outlineSize=2
      outlineColor=<rgba 0 0 0 1>
     }
    }
    shadowOffset=<0 0>
   }
   slot {class=HorizontalBoxSlot, properties {verticalAlignment=VAlign_Center, padding=<ltrb 8 0 0 0>}}
  }
 ]
}
