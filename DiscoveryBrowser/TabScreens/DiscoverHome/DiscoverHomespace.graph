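# DiscoverHomespace is the Discover home screen in the frontend lobby: a sliding blade that moves between play lobby, homebar peek and fullscreen browsing.
# List scroll, swipes, mouse wheel and activation drive the blade, guarded by lobby modal checks.
# Blade delegates trigger expand, fade and homebar animations, notify AthenaLobby and adjust list padding.
# It also handles tab background colors, analytics, library coach marks, camera transitions, download and playtime state.
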WidgetBlueprint /DiscoveryBrowser/TabScreens/DiscoverHome/DiscoverHomespace
animations [
 Anim_DiscoverHome_Expand {
  binding_count=6
  length_seconds=0.2
  object=&self:Anim_DiscoverHome_Expand
  sequence {
   bindings [
    Image_TopFade {
     id=#1
     possessed_class=Image
     tracks [{
       type=MovieSceneFloatTrack
       display_name=Render Opacity
       sections [{type=MovieSceneFloatSection, end_frame=12, start_frame=0, channels [None {type=MovieSceneScriptingFloatChannel, keys [[0,0],[12,1]]}]}]
      }]
    }
    BottomGradient {
     id=#2
     possessed_class=Image
     tracks [{
       type=MovieSceneColorTrack
       display_name=Color and Opacity
       sections [{
         type=MovieSceneColorSection
         end_frame=12
         start_frame=0
         channels [
          Color.R {type=MovieSceneScriptingFloatChannel}
          Color.G {type=MovieSceneScriptingFloatChannel}
          Color.B {type=MovieSceneScriptingFloatChannel}
          Color.A {type=MovieSceneScriptingFloatChannel, keys [[0,1],[12,0]]}
         ]
        }]
      }]
    }
    Image_BottomFade {
     id=#3
     possessed_class=Image
     tracks [{
       type=MovieSceneFloatTrack
       display_name=Render Opacity
       sections [{type=MovieSceneFloatSection, end_frame=12, start_frame=0, channels [None {type=MovieSceneScriptingFloatChannel, keys [[0,0],[12,1]]}]}]
      }]
    }
    TopBarGradient {
     id=#4
     possessed_class=Image
     tracks [{
       type=MovieSceneFloatTrack
       display_name=Render Opacity
       sections [{type=MovieSceneFloatSection, end_frame=12, start_frame=0, channels [None {type=MovieSceneScriptingFloatChannel, keys [[0,1],[12,0]]}]}]
      }]
    }
    NebulousBackground {
     id=#5
     possessed_class=Image
     tracks [{
       type=MovieSceneWidgetMaterialTrack
       sections [{
         type=MovieSceneParameterSection
         end_frame=12
         start_frame=0
         channels [Top Gradient Control {type=MovieSceneScriptingFloatChannel, keys [[0,1],[9,0],[12,0]]}]
        }]
      }]
    }
    HomeBarBacking {
     id=#6
     possessed_class=Image
     tracks [{
       type=MovieSceneColorTrack
       display_name=Color and Opacity
       sections [{
         type=MovieSceneColorSection
         end_frame=12
         start_frame=0
         channels [
          Color.R {type=MovieSceneScriptingFloatChannel, const [1,1]}
          Color.G {type=MovieSceneScriptingFloatChannel, const [1,1]}
          Color.B {type=MovieSceneScriptingFloatChannel, const [1,1]}
          Color.A {type=MovieSceneScriptingFloatChannel, keys [[0,1],[12,0]]}
         ]
        }]
      }]
    }
   ]
   display_rate {numerator=60}
  }
 }
 Anim_NebulousBGEnterLeft {
  binding_count=1
  length_seconds=1
  object=&self:Anim_NebulousBGEnterLeft
  sequence {bindings [NebulousBackground {
     id=#7
     possessed_class=Image
     tracks [{
       type=MovieSceneWidgetMaterialTrack
       sections [
        {
         type=MovieSceneParameterSection
         end_frame=20
         start_frame=0
         channels [
          Portal Lines Left Offset {type=MovieSceneScriptingFloatChannel, keys [[0,0],[20,0.38]]}
          Portal Lines Right Offset {type=MovieSceneScriptingFloatChannel, keys [[0,0],[20,1.15]]}
          Portal Lines Left Rotation {type=MovieSceneScriptingFloatChannel, keys [[0,0.25],[20,0.15]]}
          Portal Lines Right Rotation {type=MovieSceneScriptingFloatChannel, keys [[0,0.25],[20,0.18]]}
          Radial Background Scale {type=MovieSceneScriptingFloatChannel, keys [[0,1.5],[10,4]]}
          Aurora Lines Speed {type=MovieSceneScriptingFloatChannel, const [-0.4,1]}
         ]
        }
        {
         type=MovieSceneParameterSection
         end_frame=20
         start_frame=0
         channels [Triangle X Offset {type=MovieSceneScriptingFloatChannel, keys [[0,0],[20,0.3]]}]
        }
       ]
      }]
    }]}
 }
 Anim_NebulousBGEnterRight {
  binding_count=1
  length_seconds=1
  object=&self:Anim_NebulousBGEnterRight
  sequence {bindings [NebulousBackground {
     id=#8
     possessed_class=Image
     tracks [{
       type=MovieSceneWidgetMaterialTrack
       sections [{
         type=MovieSceneParameterSection
         end_frame=20
         start_frame=0
         channels [
          Portal Lines Left Offset {type=MovieSceneScriptingFloatChannel, keys [[0,1.2],[20,0.38]]}
          Portal Lines Right Offset {type=MovieSceneScriptingFloatChannel, keys [[0,1.2],[20,1.15]]}
          Portal Lines Left Rotation {type=MovieSceneScriptingFloatChannel, keys [[0,0.25],[20,0.15]]}
          Portal Lines Right Rotation {type=MovieSceneScriptingFloatChannel, keys [[0,0.25],[20,0.18]]}
          Radial Background Scale {type=MovieSceneScriptingFloatChannel, keys [[0,1.5],[10,4]]}
         ]
        }]
      }]
    }]}
 }
 Anim_NebulousBGEnterBottom {
  binding_count=1
  length_seconds=1
  object=&self:Anim_NebulousBGEnterBottom
  sequence {bindings [NebulousBackground {
     id=#9
     possessed_class=Image
     tracks [{
       type=MovieSceneWidgetMaterialTrack
       sections [
        {
         type=MovieSceneParameterSection
         end_frame=20
         start_frame=0
         channels [
          Portal Lines Left Offset {type=MovieSceneScriptingFloatChannel, keys [[0,1.2],[20,0.38]]}
          Portal Lines Right Offset {type=MovieSceneScriptingFloatChannel, keys [[0,1.2],[20,1.15]]}
          Portal Lines Left Rotation {type=MovieSceneScriptingFloatChannel, keys [[0,0],[20,0.15]]}
          Portal Lines Right Rotation {type=MovieSceneScriptingFloatChannel, keys [[0,0],[20,0.18]]}
          Radial Background Scale {type=MovieSceneScriptingFloatChannel, keys [[0,1.5],[10,4]]}
          Aurora Lines Speed {type=MovieSceneScriptingFloatChannel, const [-0.4,1]}
         ]
        }
        {
         type=MovieSceneParameterSection
         end_frame=20
         start_frame=0
         channels [Triangle Y Offset {type=MovieSceneScriptingFloatChannel, keys [[0,0],[20,-0.3]]}]
        }
       ]
      }]
    }]}
 }
 Anim_NebulousBGExitTop {
  binding_count=1
  length_seconds=0.75
  object=&self:Anim_NebulousBGExitTop
  sequence {bindings [NebulousBackground {
     id=#10
     possessed_class=Image
     tracks [{
       type=MovieSceneWidgetMaterialTrack
       sections [
        {
         type=MovieSceneParameterSection
         end_frame=8
         start_frame=0
         channels [
          Portal Lines Left Offset {type=MovieSceneScriptingFloatChannel, keys [[0,0.38],[8,0]]}
          Portal Lines Right Offset {type=MovieSceneScriptingFloatChannel, keys [[0,1.15],[8,0]]}
          Portal Lines Left Rotation {type=MovieSceneScriptingFloatChannel, keys [[0,0.15],[8,0]]}
          Portal Lines Right Rotation {type=MovieSceneScriptingFloatChannel, keys [[0,0.18],[8,0]]}
          Radial Background Scale {type=MovieSceneScriptingFloatChannel, keys [[0,4],[8,1.5]]}
          Aurora Lines Speed {type=MovieSceneScriptingFloatChannel, const [-0.1,1]}
         ]
        }
        {
         type=MovieSceneParameterSection
         end_frame=15
         start_frame=0
         channels [Triangle Y Offset {type=MovieSceneScriptingFloatChannel, keys [[0,0],[15,-0.2]]}]
        }
       ]
      }]
    }]}
 }
 Anim_NebulousBGExitLeft {
  binding_count=1
  length_seconds=0.3016
  object=&self:Anim_NebulousBGExitLeft
  sequence {bindings [NebulousBackground {
     id=#11
     possessed_class=Image
     tracks [{
       type=MovieSceneWidgetMaterialTrack
       sections [{
         type=MovieSceneParameterSection
         end_frame=6
         start_frame=0
         channels [
          Portal Lines Left Offset {type=MovieSceneScriptingFloatChannel, keys [[0,0.38],[4.9057,0]]}
          Portal Lines Right Offset {type=MovieSceneScriptingFloatChannel, keys [[0,1.15],[4.9057,0]]}
          Portal Lines Left Rotation {type=MovieSceneScriptingFloatChannel, keys [[0,0.15],[4.9057,0.25]]}
          Portal Lines Right Rotation {type=MovieSceneScriptingFloatChannel, keys [[0,0.18],[4.9057,0.25]]}
          Radial Background Scale {type=MovieSceneScriptingFloatChannel, keys [[0,4],[4.9057,1.5]]}
         ]
        }]
      }]
    }]}
 }
 Anim_Homebar_Reveal {
  binding_count=2
  length_seconds=0.2
  object=&self:Anim_Homebar_Reveal
  sequence {bindings [
    HomeBarBacking {
     id=#12
     possessed_class=Image
     tracks [{
       type=MovieSceneFloatTrack
       display_name=Render Opacity
       sections [{type=MovieSceneFloatSection, end_frame=4, start_frame=0, channels [None {type=MovieSceneScriptingFloatChannel, keys [[0,1],[4,0]]}]}]
      }]
    }
    WBP_HomeBarDetails {
     id=#13
     possessed_class=WBP_HomeBarDetails_C
     tracks [{
       type=MovieSceneFloatTrack
       display_name=Render Opacity
       sections [{type=MovieSceneFloatSection, channels [None {type=MovieSceneScriptingFloatChannel, keys [[0,1],[4,0]]}]}]
      }]
    }
   ]}
 }
 Anim_HomebarDetails {
  binding_count=1
  length_seconds=0.3
  object=&self:Anim_HomebarDetails
  sequence {
   bindings [WBP_HomeBarDetails {
     id=#14
     possessed_class=WBP_HomeBarDetails_C
     tracks [{
       type=MovieSceneColorTrack
       display_name=Color and Opacity
       sections [{
         type=MovieSceneColorSection
         end_frame=18
         start_frame=0
         channels [
          Color.R {type=MovieSceneScriptingFloatChannel}
          Color.G {type=MovieSceneScriptingFloatChannel}
          Color.B {type=MovieSceneScriptingFloatChannel}
          Color.A {type=MovieSceneScriptingFloatChannel, keys [[0,1],[12.078,0],[18,0]]}
         ]
        }]
      }]
    }]
   display_rate {numerator=60}
  }
 }
 Anim_DisplayFadeOverlay {
  binding_count=0
  length_seconds=5
  object=&self:Anim_DisplayFadeOverlay
  sequence {bindings []}
 }
 Anim_BackgroundColorChange {
  binding_count=1
  length_seconds=1
  object=&self:Anim_BackgroundColorChange
  sequence {bindings [DiscoverHomespace_C_17 {
     id=#15
     possessed_class=DiscoverHomespace_C
     tracks [{
       type=MovieSceneDoubleTrack
       display_name=Anim Output Background Color
       sections [{type=MovieSceneDoubleSection, end_frame=20, start_frame=0, channels [None {type=MovieSceneScriptingDoubleChannel, keys [[0,0],[20,1]]}]}]
      }]
    }]}
 }
]
graphs [
 ActivateMOTDWidget {nodes [
   ActivateMOTDWidget!() -> (then, Activate:bool=~) {is=title K2Node_FunctionEntry}
   > n1 = FindParentWidgetOfType!(execute, StartingWidget:Widget Object Reference=Self-Reference(), Type:Widget Class Reference=/Script/FortniteUI.AthenaLobbyBase) -> (then, ReturnValue:Athena Lobby Base Object Reference)
   > n2 = Cast To AthenaLobby!(execute, Object=@n1) -> (then, CastFailed, AsAthena Lobby:Athena Lobby Object Reference) {is=title K2Node_DynamicCast}
   > n3 = Cast To FortDynamicMotdTeaserWidget!(execute, Object=Get MOTDWidgetReference!(self:Athena Lobby Object Reference=@n2) -> (MOTDWidgetReference:User Widget Object Reference)) -> (then, CastFailed, AsFort Dynamic Motd Teaser Widget:Fort Dynamic Motd Teaser Widget Object Reference) {is=title K2Node_DynamicCast}
   > Set bShouldBeActive!(execute, bShouldBeActive:bool=Get Activate() -> (Activate:bool), self:Fort Dynamic Motd Teaser Widget Object Reference=@n3) -> (then, Output_Get:bool)
   > ResetActiveState(self:Fort Dynamic Motd Teaser Widget Object Reference=@n3)
  ]}
 BP_GetDesiredFocusTarget {nodes [
   Get Desired Focus Target()
   > Return Node!(
    execute
    ReturnValue:Widget Object Reference=Get BrowserList_Activities!(self:Fort Activity Browser Row View Object Reference=Get ActivityDiscoverView_NEW_VM() -> (ActivityDiscoverView_NEW_VM:Activity Discover View NEW VM Object Reference)) -> (BrowserList_Activities:Fort Activity Browser List View Object Reference)
   )
  ]}
 ContractBladeToLastPeekState {nodes [
   ContractBladeToLastPeekState() {is=title K2Node_FunctionEntry}
   n1 = Get WBP_SlidingBlade() -> (WBP_SlidingBlade:WBP Blade Object Reference)
   > Branch(Condition="Not Equal (Enum)"!(
     A:E_BladeAnimationState Enum=Get EBladeAnimState!(self:WBP Blade Object Reference=@n1) -> (EBladeAnimState:E_BladeAnimationState Enum=NewEnumerator0)
     B:E_BladeAnimationState Enum=NewEnumerator0
    ) -> (ReturnValue:bool))
   > SendHomespaceSubviewEvent!(
    execute
    self:Fort Discover Homespace Object Reference
    SubWidget:Widget Object Reference=Get WBP_HomeBarDetails() -> (WBP_HomeBarDetails:WBP Home Bar Details Object Reference)
    Text:String=Homebar
   ) -> (then)
   > OnBladeContract(self:WBP Blade Object Reference=Get WBP_SlidingBlade())
   > SelectCurrentActivityTile()
   > SetBottomButtonVisibility!(
    execute
    self:WBP Bottom Bar Decoy Button Object Reference=Get Button_BackToTop_BottomBarHack() -> (Button_BackToTop_BottomBarHack:WBP Bottom Bar Decoy Button Object Reference)
    ShouldButtonBeVisible:bool=false
   ) -> (then)
   > UpdateMiscActionBindingVisibility!(execute, self:Fort Activity Discover View V2 Object Reference=Get ActivityDiscoverView_NEW_VM(), bVisible:bool=false) -> (then)
   > TransitionToFullscreen!(execute, self:Scrollable Activatable Widget Object Reference, bIsFullscreen:bool=false) -> (then)
   > OnTransitionDiscoverFullscreen!(execute, self:Activity Discover View NEW VM Object Reference=Get ActivityDiscoverView_NEW_VM(), IsDiscoverFullscreen:bool=false) -> (then)
   > Branch(Condition=IsActivated!(self:Common Activatable Widget Object Reference) -> (ReturnValue:bool))
   > n2 = Branch(Condition=Get StartedTransitionToSecondPeekState!(self:WBP Blade Object Reference=Get WBP_SlidingBlade()) -> (StartedTransitionToSecondPeekState:bool))
   n3 = Get Anim_Homebar_Reveal() -> (Anim_Homebar_Reveal:Widget Animation Object Reference)
   > PlayAnimationReverse(InAnimation=@n3)
   PlayAnimationForward(execute=@n2.else, InAnimation=@n3)
  ]}
 EventGraph {nodes [
   n1 = "On Clicked (Button_BackToTop_BottomBarHack)"!() -> (OutputDelegate, then, Button:Common Button Base Object Reference) {is=title K2Node_ComponentBoundEvent}
   ContractBladeToLastPeekState(execute=[@n1.then, @n3.then, @n4.then]) {is=title K2Node_CallFunction}
   > n2 = Branch(Condition=GetIsDiscoverReduxEnabled!() -> (bIsEnabled:bool))
   n3 = "Broadcast Search Back Button Clicked (ActivityDiscoverView_NEW_VM)"!() -> (OutputDelegate, then, ButtonObject:WBP Bottom Bar Decoy Button Object Reference) {is=title K2Node_ComponentBoundEvent}
   SetBottomButtonVisibility(execute=@n2.then, self=@n3.ButtonObject, ShouldButtonBeVisible=false)
   > Reset Browser List Padding!(execute:Exec, Activity Discovery View VM:Activity Discover View NEW VM Object Reference=Get ActivityDiscoverView_NEW_VM()) {is=title K2Node_MacroInstance}
   n4 = "Navigate Up Out Of Search Row (ActivityDiscoverView_NEW_VM)"() {is=title K2Node_ComponentBoundEvent}
   n5 = "Broadcast Tab Search Selected Index (ActivityDiscoverView_NEW_VM)"!() -> (OutputDelegate, then, SelectedTabIndex:int=0) {is=title K2Node_ComponentBoundEvent}
   SetBackgroundColorEnums!(execute=[@n5.then, @n66.then], self, Tab Index:int=@n5.SelectedTabIndex) -> (then)
   > SetAnim_OutputBackgroundColor!(execute, self, Anim Progress:double=Get Anim_OutputBackgroundColor() -> (Anim_OutputBackgroundColor:double)) -> (then)
   > PlayAnimation(InAnimation=Get Anim_BackgroundColorChange() -> (Anim_BackgroundColorChange:Widget Animation Object Reference))
   > Set CachedTabIndex(CachedTabIndex:int=@n5.SelectedTabIndex) -> (Output_Get:int=0)
   n6 = TopLevelTabsChanged!() -> (OutputDelegate, then, LeavingFeature:EFortUIFeature Enum=ShowHome, EnteringFeature:EFortUIFeature Enum=ShowHome) {is=title K2Node_CustomEvent}
   > n7 = Branch(Condition="Equal (Enum)"(A:EFortUIFeature Enum=@n6.LeavingFeature, B:EFortUIFeature Enum=ShowPlay))
   n8 = IsActivated(self=Get ActivityDiscoverView_NEW_VM())
   > Branch(Condition=@n8)
   > n9 = GetConsoleVariableBoolValue!(execute, VariableName:String=Athena.DiscoverSurface.FireDiscoverTabAnalyticEventsOnTopNavTabChange) -> (then, ReturnValue:bool)
   > n10 = Branch(Condition=@n9)
   n11 = Get ActivityDiscoverView_NEW_VM()
   > EndDiscoverAnalytics(self:Activity Discover View NEW VM Object Reference=@n11)
   FireDiscoverExitedAnalyticEvent(execute=@n10.else, self:Fort Activity Discover View V2 Object Reference=@n11)
   Branch(execute=@n7.else, Condition="Equal (Enum)"(A=@n6.EnteringFeature, B=ShowPlay))
   > Branch(Condition=@n8)
   > n12 = GetConsoleVariableBoolValue(VariableName=Athena.DiscoverSurface.FireDiscoverTabAnalyticEventsOnTopNavTabChange)
   > n13 = Branch(Condition=@n12)
   > StartDiscoverAnalytics(self:Activity Discover View NEW VM Object Reference=@n11)
   StartDiscoverAnalyticSession(execute=@n13.else, self:Fort Activity Discover View V2 Object Reference=@n11)
   n14 = On Library Tile Released() {is=title K2Node_CustomEvent}
   DeactivateWidget(
    execute=[@n14.then, @n22.then]
    self=Get WBP_Discover_Tile_Library_CoachMark() -> (WBP_Discover_Tile_Library_CoachMark:WBP Discover Tile Library Coach Mark Object Reference)
   )
   > DeactivateWidget(self=Get WBP_Discover_Tile_Library_DownloadStarted_CoachMark() -> (WBP_Discover_Tile_Library_DownloadStarted_CoachMark:WBP Discover Tile Library Download Started Coach Mark Object Reference))
   > Set ShouldDisplayLibraryTooltip(ShouldDisplayLibraryTooltip:bool=false) -> (Output_Get:bool)
   On Activate Library Tooltip() {is=title K2Node_CustomEvent}
   > Branch(Condition=AND Boolean(A=Get IsShowingLibraryTile() -> (IsShowingLibraryTile:bool), B=Get ShouldDisplayLibraryTooltip() -> (ShouldDisplayLibraryTooltip:bool)))
   > ActivateWidget(self=Get WBP_Discover_Tile_Library_CoachMark())
   > Branch(Condition=IsActivated(self=Get WBP_Discover_Tile_Library_DownloadStarted_CoachMark()))
   > DeactivateWidget(self=Get WBP_Discover_Tile_Library_DownloadStarted_CoachMark())
   OnBladeResized() {is=title K2Node_CustomEvent}
   > RefreshPeekPushPercentage(self:Fort Discover Homespace Object Reference)
   OnHomebarRevealAnimationFinished() {is=title K2Node_CustomEvent}
   n15 = "Equal (Enum)"(
    A:EPlaytimeState Enum=Get PlayState!(self:Player Playtime View Model Object Reference=Get PlayerPlaytimeVM() -> (PlayerPlaytimeVM:Player Playtime View Model Object Reference)) -> (PlayState:EPlaytimeState Enum=Unblocked)
    B:EPlaytimeState Enum=PlayTime_Blocked
   )
   n16 = Get IsFullscreen!(self:Activity Discover View NEW VM Object Reference=Get ActivityDiscoverView_NEW_VM()) -> (IsFullscreen:bool)
   > Branch(Condition=AND Boolean!(A=GetIsDiscoverReduxEnabled(), B=NOT Boolean(A=@n16), C:bool=NOT Boolean(A=@n15)) -> (ReturnValue))
   > ActivateMOTDWidget!(execute, self, Activate:bool=true) -> (then) {is=title K2Node_CallFunction}
   OnBladePeekStateAnimationFinished() {is=title K2Node_CustomEvent}
   > n17 = Branch(Condition=IsActivated())
   > Branch(Condition=ShouldShowOnboarding!(self:Fort Sidebar Onboard Tooltip Widget Object Reference=Get WBP_Discover_Tile_Library_CoachMark()) -> (ReturnValue:bool))
   > On Activate Library Tooltip() {is=title K2Node_CallFunction}
   n18 = Get ActivityDiscoverView_NEW_VM()
   DeactivationTransitionCompleted(execute=@n17.else, self:Fort Activity Discover View V2 Object Reference=@n18)
   > Reset Browser List Padding(Activity Discovery View VM=@n18)
   n19 = On Entry Generated!() -> (OutputDelegate, then, Widget:User Widget Object Reference) {is=title K2Node_CustomEvent}
   > n20 = Cast To WBP_Discover_Tile_Library!(execute, Object=@n19.Widget) -> (then, CastFailed, AsWBP Discover Tile Library:WBP Discover Tile Library Object Reference) {is=title K2Node_DynamicCast}
   > Set IsShowingLibraryTile(IsShowingLibraryTile:bool=true) -> (Output_Get:bool)
   > n21 = Bind Event to On Library Tile Released(self:WBP Discover Tile Library Object Reference=@n20, Delegate=@n14) {is=title K2Node_AddDelegate}
   n22 = On Library Tile Clicked() {is=title K2Node_CustomEvent}
   Bind Event to On Library Tile Clicked(execute=@n21.then, self:WBP Discover Tile Library Object Reference=@n20, Delegate=@n22) {is=title K2Node_AddDelegate}
   OnInputMethodChanged!() -> (OutputDelegate, then, InputType:ECommonInputType Enum) {is=title K2Node_CustomEvent}
   > UpdateSlidingBladeClickCatcherAvailability()
   n23 = OnFirstScrollOutOfDisco() {is=title K2Node_CustomEvent}
   n24 = Get ActivityDiscoverView_NEW_VM()
   > Unbind Event from On Transition Discover Out(self:Activity Discover View NEW VM Object Reference=@n24, Delegate=@n23) {is=title K2Node_RemoveDelegate}
   > Branch(Condition=Get bIsActive() -> (bIsActive:bool))
   > TransitionToHomebar()
   > SetPadding!(
    execute
    self:Overlay Slot Object Reference=SlotAsOverlaySlot(Widget=Get BrowserList_Activities(self=@n24))
    InPadding_Left:float=0
    InPadding_Top:float=30
    InPadding_Right:float=0
    InPadding_Bottom:float=0
   ) -> (then)
   OnBladeContractExpandAnimationFinished() {is=title K2Node_CustomEvent}
   > SetDiscoverListScrollingAllowed()
   > HandleVisibleEntriesChanged(self:Fort Activity Browser List View Object Reference=Get BrowserList_Activities(self=Get ActivityDiscoverView_NEW_VM()))
   > NotifyGridRowBladeTransitionFinished()
   OnBladePeekButtonClicked() {is=title K2Node_CustomEvent}
   > ActivateWidget()
   n25 = OnListScrolled!() -> (OutputDelegate, then, ItemOffset:double=~, DistanceRemaining:double=~) {is=title K2Node_CustomEvent}
   n26 = Get EBladeAnimState(self=Get WBP_SlidingBlade()) -> (EBladeAnimState=NewEnumerator0)
   > n27 = Branch(Condition=AND Boolean!(A="Not Equal (Enum)"(A=@n26, B=NewEnumerator0), B="Equal (Float)"(A=@n25.ItemOffset)) -> (ReturnValue))
   > Set CurrentScrollOffset(CurrentScrollOffset:double=0) -> (Output_Get:double)
   > n28 = Branch(Condition=IsUsingTouch(PlayerController=GetOwningPlayer()))
   n29 = Get EBladeAnimState(self=Get WBP_SlidingBlade()) -> (EBladeAnimState=NewEnumerator0)
   n30 = Get WBP_SlidingBlade()
   n31 = Branch(
    execute=[@n60.then, @n28.else]
    Condition=AND Boolean(
     A="Equal (Enum)"(A:E_BladeAnimationState Enum=@n29, B:E_BladeAnimationState Enum=NewEnumerator0)
     B=Get FinishedTransitionToSecondPeekState!(self:WBP Blade Object Reference=@n30) -> (FinishedTransitionToSecondPeekState:bool)
    )
   )
   > DeactivateWidget()
   > n32 = FindParentWidgetOfType(StartingWidget=Self-Reference(), Type=&AthenaLobbyBase)
   > Is Valid(InputObject=@n32)
   > SendHomespaceSubviewEvent(SubWidget=@n32, Text=Play)
   n33 = Get CurrentScrollOffset() -> (CurrentScrollOffset:double)
   n34 = Get EBladeAnimState(self=Get WBP_SlidingBlade()) -> (EBladeAnimState=NewEnumerator0)
   Branch(
    execute=@n31.else
    Condition=AND Boolean(A="Not Equal (Enum)"(A=@n34, B=NewEnumerator0), B="Nearly Equal (Float)"(A=@n33, B=0, ErrorTolerance=0.01))
   )
   > ContractBladeToLastPeekState()
   > Branch(Condition=GetIsDiscoverReduxEnabled())
   > Reset Browser List Padding(Activity Discovery View VM=Get ActivityDiscoverView_NEW_VM())
   n35 = Branch(execute=@n27.else, Condition="float > float"(A=@n25.ItemOffset, B=Get CurrentScrollOffset()))
   > n36 = Set CurrentScrollOffset(CurrentScrollOffset=@n25.ItemOffset)
   n37 = Branch(execute=[@n65.then, @n36.then], Condition=NOT Boolean(A=IsActivated()))
   > ActivateWidget()
   > SendHomespaceSubviewEvent(SubWidget=Get WBP_HomeBarDetails(), Text=Homebar)
   n38 = "Not Equal (Enum)"(A=Get EBladeAnimState(self=Get WBP_SlidingBlade()) -> (EBladeAnimState=NewEnumerator0), B=NewEnumerator4)
   Branch(execute=@n37.else, Condition=@n38)
   > ScrollDownToDisco!(execute, self, InPadding_Left:float=0, InPadding_Top:float=-30, InPadding_Right:float=0, InPadding_Bottom:float=0) -> (then)
   Set CurrentScrollOffset(execute=@n35.else, CurrentScrollOffset=@n25.ItemOffset)
   n39 = On Browser List Activities Entry Generated!() -> (OutputDelegate, then, Widget:User Widget Object Reference) {is=title K2Node_CustomEvent}
   > n40 = Cast To ActivityBrowserDiscoverItemRow_Homebar!(execute, Object=@n39.Widget) -> (then, CastFailed, AsActivity Browser Discover Item Row Homebar:Activity Browser Discover Item Row Homebar Object Reference) {is=title K2Node_DynamicCast}
   > Bind Event to On Entry Generated(
    self:List View Base Object Reference=Get ListView_Tiles!(self:Fort Discover Item Browser Row Object Reference=@n40) -> (ListView_Tiles:Fort Discover Item List View Object Reference)
    Delegate=@n19
   ) {is=title K2Node_AddDelegate}
   > Set HomebarWidget(HomebarWidget:Widget Object Reference=@n40) -> (Output_Get:Widget Object Reference)
   OnBladeContractStart() {is=title K2Node_CustomEvent}
   n41 = Get Anim_DiscoverHome_Expand() -> (Anim_DiscoverHome_Expand:Widget Animation Object Reference)
   > PlayAnimationReverse(InAnimation=@n41)
   n42 = Get Anim_HomebarDetails() -> (Anim_HomebarDetails:Widget Animation Object Reference)
   > PlayAnimationReverse(InAnimation=@n42)
   > SetDiscoverListScrollingAllowed()
   > GetLobbyWidgetThenNotifyFullscreen!(execute, self, IsFullScreen:bool=false) -> (then)
   > UpdateOverlayBladeContentPadding!(execute, self, IsContracted:bool=true) -> (then)
   OnBladeExpandStart() {is=title K2Node_CustomEvent}
   > PlayAnimationForward(InAnimation=@n41)
   > PlayAnimationForward(InAnimation=@n42)
   > GetLobbyWidgetThenNotifyFullscreen(IsFullScreen=true)
   > UpdateOverlayBladeContentPadding(IsContracted=false)
   Event PreConstruct()
   > SetRenderOpacity(
    self=Get FullBackplate_Image!(self:WBP Blade Object Reference=Get WBP_SlidingBlade()) -> (FullBackplate_Image:Image Object Reference)
    InOpacity=0
   )
   > SetDynamicMaterials()
   > SetHomebarPeekRatios()
   > SetScalarParameterValue(self=Get Background MID() -> (Background MID:Material Instance Dynamic Object Reference), ParameterName=Top Gradient Control, Value=1)
   > n43 = GetSafeZonePadding!(execute) -> (
    then
    SafePadding_X:double
    SafePadding_Y:double
    SafePadding_Z:double
    SafePadding_W:double
    SafePaddingScale:Vector 2D Structure
    SpillOverPadding:Vector 4 Structure
   )
   > Set bSafezoneSidesEnabled(bSafezoneSidesEnabled:bool=AND Boolean(A="float > float"(A=@n43), B="float > float"(A=@n43.SafePadding_Z))) -> (Output_Get:bool)
   > n44 = Set bSafezoneTopEnabled(bSafezoneTopEnabled:bool="float > float"(A=@n43.SafePadding_Y)) -> (Output_Get:bool)
   n45 = Get MaskFadePowerParameterMobile() -> (MaskFadePowerParameterMobile:double)
   n46 = Get MaskFadePowerParameterMobileSafezone() -> (MaskFadePowerParameterMobileSafezone:double)
   > SetScalarParameterValue(
    self=Get BackgroundTop MID() -> (BackgroundTop MID:Material Instance Dynamic Object Reference)
    ParameterName=Mask Fade Power
    Value=SelectFloat(
     A=SelectFloat(A=@n46, B=@n45, bPickA=@n44)
     B=Get Mask Fade Power Parameter() -> (Mask Fade Power Parameter:float)
     bPickA=IsUsingCompactUI()
    )
   )
   > SetMobileStyle() {is=title K2Node_CallFunction}
   Event Construct()
   n47 = Get WBP_SlidingBlade()
   > Assign On Notify Parent Blade Contract Start!(execute:Exec, self:WBP Blade Object Reference=@n47, Delegate:"Delegate (by ref)"=Create Event()) -> (then:Exec) {is=title K2Node_AssignDelegate}
   > Assign On Notify Parent Blade Expand Start!(execute:Exec, self:WBP Blade Object Reference=@n47, Delegate:"Delegate (by ref)"=Create Event()) -> (then:Exec) {is=title K2Node_AssignDelegate}
   > Bind Event to On Notify Parent Peek Button Clicked(self:WBP Blade Object Reference=@n47, Delegate=Create Event()) {is=title K2Node_AddDelegate}
   > Bind Event to On Notify Parent Blade Transition Finished(self:WBP Blade Object Reference=@n47, Delegate=Create Event()) {is=title K2Node_AddDelegate}
   > Bind Event to On Notify Parent Blade Peek State Transition Finished(self:WBP Blade Object Reference=@n47, Delegate=Create Event()) {is=title K2Node_AddDelegate}
   > Bind Event to On Notify Blade Resized(self:WBP Blade Object Reference=@n47, Delegate=Create Event()) {is=title K2Node_AddDelegate}
   > Bind Event to On Entry Generated(self=Get BrowserList_Activities(self=Get ActivityDiscoverView_NEW_VM()), Delegate=Create Event())
   n48 = Get ActivityDiscoverView_NEW_VM()
   > Bind Event to On List Scrolled(self:Activity Discover View NEW VM Object Reference=@n48, Delegate=Create Event()) {is=title K2Node_AddDelegate}
   > Bind Event to Broadcast to Homespace Tile Clicked(self:Activity Discover View NEW VM Object Reference=@n48, Delegate=Create Event()) {is=title K2Node_AddDelegate}
   > Bind Event to On Activity Requested(self:Fort Activity Discover View V2 Object Reference=@n48, Delegate=Create Event()) {is=title K2Node_AddDelegate}
   > BindToAnimationFinished(Animation=Get Anim_Homebar_Reveal(), Delegate=Create Event())
   > SetDiscoverListScrollingAllowed()
   > Assign On Input Method Changed!(execute:Exec, self:Common Input Object Reference=Get CommonInputSubsystem(), Delegate:"Delegate (by ref)"=Create Event()) -> (then:Exec) {is=title K2Node_AssignDelegate}
   > UpdateSlidingBladeClickCatcherAvailability()
   > n49 = GetShouldImmediateTransitionToHomebar!(execute, self) -> (then, ShouldImmediateTransition:bool)
   > SetShouldStartInSecondPeekState!(execute, self:WBP Blade Object Reference=Get WBP_SlidingBlade(), ShouldStartInSecondPeekState:bool=@n49) -> (then)
   > SetRenderOpacity(self=Get WBP_HomeBarDetails(), InOpacity=Select(Option 0:float=0, Option 1:float=1, Index=@n49) -> (ReturnValue:float))
   > UpdateVisualForArabic()
   > BindToTabChange!(execute, Context:Widget Object Reference=Self-Reference(), EventToBind:"Delegate (by ref)"=@n6) -> (then)
   Event On Activated()
   > Init(self:Player Playtime View Model Object Reference=Get PlayerPlaytimeVM())
   > ActivateWidget(self=Get ActivityDiscoverView_NEW_VM())
   > SelectCurrentActivityTile()
   > n50 = Branch(Condition="ShouldAutoScrollToDisco "!() -> (Autoscroll:bool))
   > ScrollDownToDisco(InPadding_Left=0, InPadding_Top=-200, InPadding_Right=0, InPadding_Bottom=0)
   > Bind Event to On Transition Discover Out(self:Activity Discover View NEW VM Object Reference=@n24, Delegate=@n23) {is=title K2Node_AddDelegate}
   > n51 = FindParentWidgetOfType(StartingWidget=Self-Reference(), Type=&AthenaLobbyBase)
   > SetHomeBarShownOnActivated(self:Athena Lobby Base Object Reference=@n51)
   > DelayUntilNextFrame!(execute) -> (then)
   > SetDiscoveryLandingActionButtons()
   TransitionToHomebar(execute=@n50.else)
   Event On Deactivated()
   > DeactivateWidget(self=Get ActivityDiscoverView_NEW_VM())
   > ContractBladeToLastPeekState()
   > TransitionToPlayLobby()
   > SelectCurrentActivityTile()
   n52 = Get ActivityDiscoverView_NEW_VM()
   > SetPadding(
    self=SlotAsOverlaySlot(Widget=Get BrowserList_Activities(self=@n52))
    InPadding_Left=0
    InPadding_Top=0
    InPadding_Right=0
    InPadding_Bottom=0
   )
   n53 = Event ShowDownloadStartedTooltip!() -> (OutputDelegate, then, TooltipDescription:String) {is=title K2Node_Event}
   n54 = IsActivated(self=Get WBP_Discover_Tile_Library_CoachMark())
   > Branch(Condition=AND Boolean(A=Get IsShowingLibraryTile(), B=NOT Boolean(A=@n54)))
   n55 = Get WBP_Discover_Tile_Library_DownloadStarted_CoachMark()
   > SetTextDescription!(
    execute
    self:WBP UIKit Coach Mark Base Object Reference=Get WBP_UIKit_CoachMark!(self:WBP Discover Tile Library Download Started Coach Mark Object Reference=@n55) -> (WBP_UIKit_CoachMark:WBP UIKit Coach Mark Object Reference)
    Text:Text="To Text (String)"(InString=@n53.TooltipDescription)
   ) -> (then, Empty:bool)
   > ActivateWidget(self=@n55)
   Event HandleSwipeUp() {is=title K2Node_Event}
   n56 = Get BrowserList_Activities(self=Get ActivityDiscoverView_NEW_VM())
   > n57 = Branch(Condition="Nearly Equal (Float)"(A=GetScrollOffset!(self:List View Base Object Reference=@n56) -> (ReturnValue:float), B=0, ErrorTolerance=0.01))
   n58 = Get EBladeAnimState(self=Get WBP_SlidingBlade()) -> (EBladeAnimState=NewEnumerator0)
   Branch(
    execute=[@n72.Pressed, @n57.then]
    Condition=OR Boolean(A="Not Equal (Enum)"(A=@n58, B=NewEnumerator2), B="Not Equal (Enum)"(A=@n58, B=NewEnumerator1))
   )
   > n59 = GetIsLobbyActivatedWithNoModals!(execute, self) -> (then, IsLobbyActivated:bool)
   > n60 = Branch(Condition=@n59)
   Event HandleSwipeDown() {is=title K2Node_Event}
   > n61 = Branch(Condition=Get IsFullscreen(self=Get ActivityDiscoverView_NEW_VM()))
   n62 = GetShouldSkipMouseWheelDown!(execute=[@n71.else, @n61.else, @n68.else], self) -> (then, ShouldSkipMouseWheelDown:bool)
   > n63 = Branch(Condition=@n62)
   n64 = GetIsLobbyActivatedWithNoModals(execute=@n63.else)
   > n65 = Branch(Condition=@n64)
   Event Destruct()
   > Unbind Event from On Notify Blade Resized(self:WBP Blade Object Reference=Get WBP_SlidingBlade(), Delegate=Create Event()) {is=title K2Node_RemoveDelegate}
   > UnbindFromAnimationFinished!(
    execute
    self:User Widget Object Reference
    Animation:Widget Animation Object Reference=Get Anim_Homebar_Reveal()
    Delegate:Delegate=Create Event()
   ) -> (then)
   > Unbind Event from On Input Method Changed(self=Get CommonInputSubsystem(), Delegate=Create Event())
   n66 = Event OnInitialized()
   Mouse Wheel Down!() -> (Pressed:Exec, Released:Exec, Key:Key Structure) {is=title K2Node_InputKey, inputKey=MouseScrollDown}
   n67 = IsHovered(self=Get ActivityDiscoverView_NEW_VM())
   > n68 = Branch(Condition=AND Boolean(A="Not Equal (Boolean)"(A=@n67, B=true) -> (ReturnValue=~), B=@n38))
   > n69 = GetIsLobbyActivatedWithNoModals()
   > Branch(Condition=@n69)
   > n70 = ScrollListViewToNextRow!(execute, self:Fort Activity Browser List View Object Reference=Get BrowserList_Activities(self=Get ActivityDiscoverView_NEW_VM())) -> (then, ReturnValue:bool)
   > n71 = Branch(Condition=@n70)
   n72 = Mouse Wheel Up!() -> (Pressed:Exec, Released:Exec, Key:Key Structure) {is=title K2Node_InputKey, inputKey=MouseScrollUp}
   OnHomespaceTileClicked() {is=title K2Node_CustomEvent}
  ]}
 GetHomebarDetailsSubtextVisibility {nodes [
   n1 = GetHomebarDetailsSubtextVisibility!() -> (then, IsLibrary:bool=~) {is=title K2Node_FunctionEntry}
   > Return Node!(
    execute
    Visibility:ESlateVisibility Enum=Select(Option 0:ESlateVisibility Enum=Collapsed, Option 1:ESlateVisibility Enum=SelfHitTestInvisible, Index=@n1) -> (ReturnValue:ESlateVisibility Enum)
   )
  ]}
 GetIsDiscoverReduxEnabled {nodes [
   GetIsDiscoverReduxEnabled() {is=title K2Node_FunctionEntry}
   > n1 = Cast To FortPlayerController(Object=GetOwningPlayer())
   > Return Node!(
    execute
    bIsEnabled:bool=GetDiscoveryReduxExperimentValue!(FortPlayerController:Player Controller Object Reference=@n1) -> (ReturnValue:bool)
   )
  ]}
 GetIsLobbyActivatedWithNoModals {nodes [
   GetIsLobbyActivatedWithNoModals() {is=title K2Node_FunctionEntry}
   > n1 = FindParentWidgetOfType(StartingWidget=Self-Reference(), Type=&AthenaLobbyBase)
   n2 = IsAnyFauxModelOpen!(self:Fort Lobby Widget Object Reference=@n1) -> (ReturnValue:bool)
   n3 = IsBladesMenuActive!(self:Athena Lobby Base Object Reference=@n1) -> (ReturnValue:bool)
   > Return Node!(execute, IsLobbyActivated:bool=AND Boolean!(A=NOT Boolean(A=@n3), B=IsActivated(self=@n1), C:bool=NOT Boolean(A=@n2)) -> (ReturnValue))
  ]}
 GetLobbyWidgetThenNotifyFullscreen {nodes [
   n1 = GetLobbyWidgetThenNotifyFullscreen!() -> (then, IsFullScreen:bool=~) {is=title K2Node_FunctionEntry}
   > n2 = FindParentWidgetOfType(StartingWidget=Self-Reference(), Type=&AthenaLobbyBase)
   > n3 = Cast To AthenaLobby(Object=@n2)
   > OnHomespaceFullScreenStateChanged!(execute, self:Athena Lobby Object Reference=@n3, IsFullScreen:bool=@n1) -> (then)
  ]}
 GetLobbyWidgetThenTransition {nodes [
   n1 = GetLobbyWidgetThenTransition!() -> (then, IsOnPlayView:bool=~) {is=title K2Node_FunctionEntry}
   > n2 = FindParentWidgetOfType(StartingWidget=Self-Reference(), Type=&AthenaLobbyBase)
   > n3 = Cast To AthenaLobby(Object=@n2)
   > OnHomespaceTellLobbyToTransition!(execute, self:Athena Lobby Object Reference=@n3, PlayVisible:bool=@n1) -> (then)
   n4 = Get FortFrontEndContext!() -> (ReturnValue:Fort Front End Context Object Reference) {is=title K2Node_GetBlueprintContextEx}
   > SetFrontEndCamera!(
    execute
    self:Fort Front End Context Object Reference=Get FortFrontEndContext()
    NewState:EFrontEndCamera Enum=Select(
     Option 0:EFrontEndCamera Enum=Get LocalHomebarCamera() -> (LocalHomebarCamera:EFrontEndCamera Enum=Invalid)
     Option 1:EFrontEndCamera Enum=GetFrontEndHeroCamera!(self:Fort Front End Context Object Reference=@n4) -> (ReturnValue:EFrontEndCamera Enum=Invalid)
     Index=@n1
    ) -> (ReturnValue:EFrontEndCamera Enum)
    TransitionParams:View Target Transition Params Structure=Make View Target Transition Params!(BlendTime:float=0.25, BlendFunction:EViewTargetBlendFunction Enum=VTBlend_EaseOut, BlendExp:float=2, bLockOutgoing:bool=False) -> (ViewTargetTransitionParams:View Target Transition Params Structure) {is=title K2Node_MakeStruct}
   ) -> (then)
  ]}
 GetShouldImmediateTransitionToHomebar {nodes [
   GetShouldImmediateTransitionToHomebar() {is=title K2Node_FunctionEntry}
   > n1 = FindParentWidgetOfType(StartingWidget=Self-Reference(), Type=&AthenaLobbyBase)
   > n2 = ShouldShowHomeBarOnActivated!(execute, self:Athena Lobby Base Object Reference=@n1) -> (then, ReturnValue:bool)
   > Return Node!(execute, ShouldImmediateTransition:bool=@n2)
  ]}
 GetShouldSkipMouseWheelDown {nodes [
   GetShouldSkipMouseWheelDown() {is=title K2Node_FunctionEntry}
   n1 = Get FinishedTransitionToSecondPeekState(self=Get WBP_SlidingBlade())
   > Return Node!(execute, ShouldSkipMouseWheelDown:bool=AND Boolean!(A=IsActivated(), B=NOT Boolean(A=@n1)) -> (ReturnValue))
  ]}
 HandleDownloadProgressChanged {nodes [
   HandleDownloadProgressChanged!() -> (then, DownloadProgress:double=~) {is=title K2Node_FunctionEntry}
   n1 = Get DiscoverSelectedActivityViewModel() -> (DiscoverSelectedActivityViewModel:Discover Selected Activity View Model Object Reference)
   n2 = Get ActivityVM!(self:Discover Selected Activity View Model Object Reference=@n1) -> (ActivityVM:Discover Activity View Model Object Reference)
   > Branch(Condition=AND Boolean(A=IsValid(Object=@n1), B=IsValid(Object=@n2)))
   > Branch(Condition="Equal (Enum)"(
     A:EFortActivityViewModelDownloadState Enum=GetDownloadState!(self:Discover Activity View Model Object Reference=@n2) -> (ReturnValue:EFortActivityViewModelDownloadState Enum=Unknown)
     B:EFortActivityViewModelDownloadState Enum=Downloading
    ))
   > UpdateDownloadProgressText!(
    execute
    self:WBP Home Bar Details Object Reference=Get WBP_HomeBarDetails()
    DownloadProgress:double=Get DownloadProgress() -> (DownloadProgress:double)
   ) -> (then)
  ]}
 HandleDownloadStateChanged {nodes [
   n1 = HandleDownloadStateChanged!() -> (then, DownloadState:EFortActivityViewModelDownloadState Enum=Unknown) {is=title K2Node_FunctionEntry}
   > Branch(Condition="Not Equal (Enum)"(A:EFortActivityViewModelDownloadState Enum=@n1, B:EFortActivityViewModelDownloadState Enum=Downloading))
   > SetupActivityMessage()
  ]}
 HandleInvalidActivityReasonChanged {nodes [
   HandleInvalidActivityReasonChanged!() -> (then, InvalidActivityReason:EFortInvalidActivityReason Enum=None) {is=title K2Node_FunctionEntry}
   > SetupActivityMessage()
  ]}
 HandleNavigateUpFromDiscover {nodes [
   HandleNavigateUpFromDiscover!() -> (then, Navigation:EUINavigation Enum) {is=title K2Node_FunctionEntry}
   > DeactivateWidget()
   > Return Node!(execute, ReturnValue:Widget Object Reference)
  ]}
 HandleOnActivityRequested {nodes [
   HandleOnActivityRequested() {is=title K2Node_FunctionEntry}
   > DeactivateWidget()
  ]}
 HandlePlaytimeStateChanged {nodes [
   n1 = HandlePlaytimeStateChanged!() -> (then, PlaytimeState:EPlaytimeState Enum=Unblocked) {is=title K2Node_FunctionEntry}
   > Branch(Condition="Equal (Enum)"(A:EPlaytimeState Enum=@n1, B:EPlaytimeState Enum=PlayTime_Blocked))
   > DeactivateWidget()
  ]}
 HandleSelectedActivityLoadingChanged {nodes [
   n1 = HandleSelectedActivityLoadingChanged!() -> (then, bLoading:bool=~) {is=title K2Node_FunctionEntry}
   > n2 = Branch(Condition=@n1)
   n3 = Get DiscoverSelectedActivityViewModel()
   Is Valid(exec=@n2.else, InputObject=@n3)
   n4 = Get bHasPartyData!(self:Discover Selected Activity View Model Object Reference=@n3) -> (bHasPartyData:bool)
   > SetActiveWidgetIndex(
    self=Get Switcher_GameDetailsContent!(self:WBP Home Bar Details Object Reference=Get WBP_HomeBarDetails()) -> (Switcher_GameDetailsContent:Common Activatable Widget Switcher Object Reference)
    Index=SelectInt(A=1, B=0, bPickA=@n4)
   )
   > SetupFriendsPlayingText()
   n5 = Get bIsLibrary!(self:Discover Selected Activity View Model Object Reference=Get DiscoverSelectedActivityViewModel()) -> (bIsLibrary:bool)
   > Set Visibility(
    Visibility=Select(Option 0:ESlateVisibility Enum=SelfHitTestInvisible, Option 1:ESlateVisibility Enum=Collapsed, Index=@n5) -> (ReturnValue:ESlateVisibility Enum)
    self=Get Switcher_GameDetailsContent(self=Get WBP_HomeBarDetails())
   )
   > n6 = Branch(Condition=@n5)
   > Setup Selected Tile For Library()
   n7 = Branch(execute=@n6.else, Condition=@n4)
   > SetupSelectedActivityForJoinableParty()
   SetupSelectedActivityForIsland(execute=@n7.else)
  ]}
 NotifyGridRowBladeTransitionFinished {nodes [
   NotifyGridRowBladeTransitionFinished() {is=title K2Node_FunctionEntry}
   n1 = Get ActivityDiscoverView_NEW_VM()
   > n2 = Set L_DisplayedEntries(L_DisplayedEntries:Array of User Widget Object References=GetDisplayedEntryWidgets!(self:List View Base Object Reference=Get BrowserList_Activities(self=@n1)) -> (ReturnValue:Array of User Widget Object References)) -> (Output_Get:Array of User Widget Object References)
   > n3 = For Each Loop(Array:Array of User Widget Object References=@n2) -> (Array Element:User Widget Object Reference)
   n4 = Get WBP_SlidingBlade()
   > SetBladeExpandContractAnimationFinished!(
    execute:Exec
    self:Object Reference=@n3
    bBladeExpandContractAnimationFinished:bool="Equal (Enum)"(A:E_BladeAnimationState Enum=Get EBladeAnimState(self=@n4) -> (EBladeAnimState=NewEnumerator0), B:E_BladeAnimationState Enum=NewEnumerator4)
   ) -> (then:Exec) {is=title K2Node_Message}
  ]}
 ResetBrowserListPadding {nodes [
   n1 = Inputs!() -> (execute, Activity Discovery View VM:Activity Discover View NEW VM Object Reference)
   > SetPadding(
    self=SlotAsOverlaySlot(Widget=Get BrowserList_Activities(self=@n1))
    InPadding_Left=0
    InPadding_Top=0
    InPadding_Right=0
    InPadding_Bottom=0
   )
   Outputs()
  ]}
 ScrollDownToDisco {nodes [
   ScrollDownToDisco!() -> (then, InPadding:Margin Structure) {is=title K2Node_FunctionEntry}
   > SendHomespaceSubviewEvent(SubWidget=Get WBP_SlidingBlade(), Text=Discover)
   > OnBladeExpand(self:WBP Blade Object Reference=Get WBP_SlidingBlade())
   > OnTransitionDiscoverFullscreen(self=Get ActivityDiscoverView_NEW_VM(), IsDiscoverFullscreen=true)
   > ActivateMOTDWidget(Activate=false)
   > TransitionToFullscreen(bIsFullscreen=true)
   > PlayAnimationForward(InAnimation=Get Anim_Homebar_Reveal())
   > SetBottomButtonVisibility(
    self=Get Button_BackToTop_BottomBarHack()
    ShouldButtonBeVisible=Select(Option 0:bool=true, Option 1:bool=false, Index=GetIsDiscoverReduxEnabled() {is=title K2Node_CallFunction}) -> (ReturnValue:bool)
   )
   > UpdateMiscActionBindingVisibility(self=Get ActivityDiscoverView_NEW_VM(), bVisible=true)
   > Branch(Condition=GetIsDiscoverReduxEnabled())
   n1 = Get ActivityDiscoverView_NEW_VM()
   > SetPadding!(
    execute
    self=SlotAsOverlaySlot(Widget=Get BrowserList_Activities(self=@n1))
    InPadding:Margin Structure=Get InPadding() -> (InPadding:Margin Structure)
   ) -> (then)
  ]}
 SelectCurrentActivityTile {nodes [
   SelectCurrentActivityTile() {is=title K2Node_FunctionEntry}
   n1 = Get ActivityDiscoverView_NEW_VM()
   > n2 = GetIndexOfCurrentActivity!(execute, self:Activity Discover View NEW VM Object Reference=@n1) -> (then, TileIndex:int=0)
   > NavigateToActivityInFirstRow!(
    execute
    self:Fort Activity Browser List View Object Reference=Get BrowserList_Activities(self=@n1)
    ActivityItemIndex:int=@n2
    bRequestNavigationToItem:bool=true
   ) -> (then)
  ]}
 SetAnim_OutputBackgroundColor {nodes [
   n1 = SetAnim_OutputBackgroundColor!() -> (then, Anim Progress:double=~) {is=title K2Node_FunctionEntry}
   n2 = Select(
    Option 0:Linear Color Structure=Select(
     Option 0=(R=0.0144,G=0.1912,B=0.3663,A=1)
     Option 1=(R=0.9387,G=0,B=0.3864,A=1)
     Index=GetConfigDataBoolWithDefault!(EntryName:String=bEnableMangoFrontEndVisuals, bDefaultValue:bool=false) -> (ReturnValue:bool)
    )
    Option 1:Linear Color Structure=(R=0.7231,G=0.0382,B=0,A=1)
    Index=GetConfigDataBoolWithDefault(EntryName=bEnableCH7FrontEndVisuals, bDefaultValue=false)
   ) -> (ReturnValue:Linear Color Structure)
   n3 = Get BackgroundColor() -> (BackgroundColor:E_Discover_Background_Color Enum=NewEnumerator0)
   n4 = Get CachedBackgroundColor() -> (CachedBackgroundColor:E_Discover_Background_Color Enum=NewEnumerator0)
   > Set OutputBackgroundColor(OutputBackgroundColor:Linear Color Structure="Lerp Using HSV (LinearColor)"!(
     A:Linear Color Structure=Select!(
      NewEnumerator0:Linear Color Structure=@n2
      NewEnumerator1:Linear Color Structure=(R=0.0697,G=0.0152,B=0.3419,A=1)
      NewEnumerator2:Linear Color Structure=(R=0.016,G=0.3278,B=0.3278,A=1)
      NewEnumerator3:Linear Color Structure=(R=0.3278,G=0.016,B=0.1199,A=1)
      NewEnumerator4:Linear Color Structure=(R=0.3278,G=0.042,B=0.016,A=1)
      Index:E_Discover_Background_Color Enum=@n4
     ) -> (ReturnValue)
     B:Linear Color Structure=Select(
      NewEnumerator0=@n2
      NewEnumerator1=(R=0.0697,G=0.0152,B=0.3419,A=1)
      NewEnumerator2=(R=0.016,G=0.3278,B=0.3278,A=1)
      NewEnumerator3=(R=0.3278,G=0.016,B=0.1199,A=1)
      NewEnumerator4=(R=0.3278,G=0.042,B=0.016,A=1)
      Index=@n3
     )
     Alpha:float=@n1
    ) -> (ReturnValue:Linear Color Structure)) -> (Output_Get:Linear Color Structure)
   > SetVectorParameterValue(
    self=[
     Get Background MID()
     Get BackgroundTop MID()
     Get BackgroundBottom MID() -> (BackgroundBottom MID:Material Instance Dynamic Object Reference)
    ]
    ParameterName=Primary Background Color
    Value=Get OutputBackgroundColor() -> (OutputBackgroundColor:Linear Color Structure)
   )
   n5 = Get BackgroundColor() -> (BackgroundColor=NewEnumerator0)
   n6 = Get CachedBackgroundColor() -> (CachedBackgroundColor=NewEnumerator0)
   > Set OutputBackgroundSecondaryColor(OutputBackgroundSecondaryColor:Linear Color Structure="Lerp Using HSV (LinearColor)"(
     A=Select(
      NewEnumerator0=(R=0,G=0,B=0.008,A=1)
      NewEnumerator1=(R=0,G=0,B=0.0078,A=1)
      NewEnumerator2=(R=0,G=0,B=0.0078,A=1)
      NewEnumerator3=(R=0,G=0,B=0.0078,A=1)
      NewEnumerator4=(R=0,G=0,B=0.0078,A=1)
      Index=@n6
     )
     B=Select(
      NewEnumerator0=(R=0,G=0,B=0.008,A=1)
      NewEnumerator1=(R=0,G=0,B=0.0078,A=1)
      NewEnumerator2=(R=0,G=0,B=0.0078,A=1)
      NewEnumerator3=(R=0,G=0,B=0.0078,A=1)
      NewEnumerator4=(R=0,G=0,B=0.0078,A=1)
      Index=@n5
     )
     Alpha=Get Anim Progress() -> (Anim Progress:double)
    )) -> (Output_Get:Linear Color Structure)
   > SetVectorParameterValue(
    self=[Get BackgroundTop MID(), Get BackgroundBottom MID(), Get Background MID()]
    ParameterName=Secondary Background Color
    Value=Get OutputBackgroundSecondaryColor() -> (OutputBackgroundSecondaryColor:Linear Color Structure)
   )
   Select!(
    Option 0:Linear Color Structure=Select(
     Option 0=(R=0.0009,G=0.0009,B=0.0097,A=1)
     Option 1=(R=0.0078,G=0.0039,B=0.0118,A=1)
     Index=GetConfigDataBoolWithDefault(EntryName=bEnableMangoFrontEndVisuals, bDefaultValue=false)
    )
    Option 1:Linear Color Structure=(R=0.0118,G=0.0235,B=0.0353,A=1)
    Index:bool=GetConfigDataBoolWithDefault(EntryName=bEnableCH7FrontEndVisuals, bDefaultValue=false)
   ) -> (ReturnValue)
  ]}
 SetBackgroundColorEnums {nodes [
   n1 = SetBackgroundColorEnums!() -> (then, Tab Index:int) {is=title K2Node_FunctionEntry}
   > Set BackgroundColor(BackgroundColor:E_Discover_Background_Color Enum=Select!(
     Option 0:E_Discover_Background_Color Enum=NewEnumerator0
     Option 1:E_Discover_Background_Color Enum=NewEnumerator1
     Option 2:E_Discover_Background_Color Enum=NewEnumerator2
     Option 3:E_Discover_Background_Color Enum=NewEnumerator3
     Option 4:E_Discover_Background_Color Enum=NewEnumerator4
     Index:int=@n1
    ) -> (ReturnValue:E_Discover_Background_Color Enum)) -> (Output_Get:E_Discover_Background_Color Enum=NewEnumerator0)
   > Set CachedBackgroundColor(CachedBackgroundColor:E_Discover_Background_Color Enum=Select(
     Option 0=NewEnumerator0
     Option 1=NewEnumerator1
     Option 2=NewEnumerator2
     Option 3=NewEnumerator3
     Option 4=NewEnumerator4
     Index=Get CachedTabIndex() -> (CachedTabIndex:int=0)
    )) -> (Output_Get:E_Discover_Background_Color Enum=NewEnumerator0)
  ]}
 SetDiscoverListScrollingAllowed {nodes [
   SetDiscoverListScrollingAllowed() {is=title K2Node_FunctionEntry}
   n1 = Get EBladeAnimState(self=Get WBP_SlidingBlade()) -> (EBladeAnimState=NewEnumerator0)
   > AllowMouseWheelScrolling!(
    execute
    self:Activity Discover View NEW VM Object Reference=Get ActivityDiscoverView_NEW_VM()
    AllowMouseWheelScrolling:bool=Select!(
     Option 0:bool=true
     Option 1:bool=false
     Index:bool="Not Equal (Enum)"(A:E_BladeAnimationState Enum=@n1, B:E_BladeAnimationState Enum=NewEnumerator4)
    ) -> (ReturnValue:bool)
   ) -> (then)
  ]}
 SetDiscoveryLandingActionButtons {nodes [
   SetDiscoveryLandingActionButtons() {is=title K2Node_FunctionEntry}
   n1 = Get ActivityDiscoverView_NEW_VM()
   > UpdateMiscActionBindingVisibility(self=@n1, bVisible=false)
   > UpdateMiscActionBindingVisibility(self=@n1, bVisible=true)
   > n2 = FindParentWidgetOfType(StartingWidget=Self-Reference(), Type=&AthenaLobbyBase)
   > n3 = Cast To AthenaLobby(Object=@n2)
   > SetBackButtonVisibility!(execute, self:Athena Lobby Object Reference=@n3, ShouldButtonBeVisible:bool=false) -> (then)
   > Set is MM Widget on Screen!(execute, is MM Widget on Screen:bool=false, self:Athena Lobby Object Reference=@n3) -> (then, Output_Get:bool)
   > SetFrontendCameraState!(execute, self:Athena Lobby Object Reference=@n3, NewState:EFrontEndCamera Enum=DiscoveryHome) -> (then)
  ]}
 SetDynamicMaterials {nodes [
   SetDynamicMaterials() {is=title K2Node_FunctionEntry}
   > n1 = GetDynamicMaterial(self=Get NebulousBackground() -> (NebulousBackground:Image Object Reference))
   > Set Background MID(Background MID:Material Instance Dynamic Object Reference=@n1) -> (Output_Get:Material Instance Dynamic Object Reference)
   > n2 = GetDynamicMaterial(self=Get Image_TopFade() -> (Image_TopFade:Image Object Reference))
   > Set BackgroundTop MID(BackgroundTop MID:Material Instance Dynamic Object Reference=@n2) -> (Output_Get:Material Instance Dynamic Object Reference)
   > n3 = GetDynamicMaterial(self=Get Image_BottomFade() -> (Image_BottomFade:Image Object Reference))
   > Set BackgroundBottom MID(BackgroundBottom MID:Material Instance Dynamic Object Reference=@n3) -> (Output_Get:Material Instance Dynamic Object Reference)
  ]}
 SetHomebarPeekRatios {nodes [
   SetHomebarPeekRatios() {is=title K2Node_FunctionEntry}
   n1 = GetViewportSize!() -> (ReturnValue_X:double, ReturnValue_Y:double)
   n2 = GetViewportScale!() -> (ReturnValue:float)
   n3 = float / float(A=SafeDivide(A=@n1, B=@n2), B=SafeDivide(A=@n1.ReturnValue_Y, B=@n2))
   > SetPeekPushPercentages!(
    execute
    self:WBP Blade Object Reference=Get WBP_SlidingBlade()
    PeekPushPercent:double=0.846
    PeekPrushPercentSecondary:double=float + float(A=float * float(A=@n3, B=-0.1105) {is=title K2Node_PromotableOperator}, B=0.8915)
   ) -> (then)
  ]}
 SetMobileStyle {nodes [
   SetMobileStyle() {is=title K2Node_FunctionEntry}
   > Branch(Condition=IsUsingCompactUI())
   > SetRenderTransform!(
    execute
    self:Widget Object Reference=Get Image_TopFade()
    InTransform_Translation:Vector 2D Structure=(X=0,Y=22)
    InTransform_Scale_X:double=1
    InTransform_Scale_Y:double=SelectFloat(A=2.2, B=1.5, bPickA=Get bSafezoneTopEnabled() -> (bSafezoneTopEnabled:bool))
    InTransform_Shear:Vector 2D Structure=(X=0,Y=0)
    InTransform_Angle:float=0
   ) -> (then)
   > UpdateOverlayBladeContentPadding(IsContracted=true)
  ]}
 Setup Selected Tile For Library {nodes [
   Setup Selected Tile For Library() {is=title K2Node_FunctionEntry}
   > "SetText (Text)"(
    self=Get Text_SelectedTile!(self:WBP Home Bar Details Object Reference=Get WBP_HomeBarDetails()) -> (Text_SelectedTile:Common Text Object Reference)
    InText="LOCTEXT(\"Library\")"
   )
  ]}
 SetupActivityMessage {nodes [
   SetupActivityMessage() {is=title K2Node_FunctionEntry}
   n1 = Get ActivityVM(self=Get DiscoverSelectedActivityViewModel())
   > SetCurrentOrInvalidReasonText!(
    execute
    self:WBP Home Bar Details Object Reference=Get WBP_HomeBarDetails()
    InvalidActivityReason:EFortInvalidActivityReason Enum=Get InvalidActivityReason!(self:Discover Activity View Model Object Reference=@n1) -> (InvalidActivityReason:EFortInvalidActivityReason Enum=None)
    IsCurrentActivity:bool=Get bIsCurrentActivity!(self:Discover Activity View Model Object Reference=@n1) -> (bIsCurrentActivity:bool)
   ) -> (then)
  ]}
 SetupFriendsPlayingText {nodes [
   SetupFriendsPlayingText() {is=title K2Node_FunctionEntry}
   n1 = Get DiscoverSelectedActivityViewModel()
   > SetFriendsCountText!(
    execute
    self:WBP Home Bar Details Object Reference=Get WBP_HomeBarDetails()
    FriendsCount:int=GetNumFriendsPlaying!(self:Discover Activity View Model Object Reference=Get ActivityVM(self=@n1)) -> (ReturnValue:int=0)
   ) -> (then)
  ]}
 SetupIslandName {nodes [
   SetupIslandName() {is=title K2Node_FunctionEntry}
   n1 = Get DiscoverSelectedActivityViewModel()
   > "SetText (Text)"(
    self=Get Text_SelectedTile(self=Get WBP_HomeBarDetails())
    InText=GetDisplayName!(self:Fort Discover Link Entry Tile Item VM Object Reference=Get ActivityVM(self=@n1)) -> (ReturnValue:Text)
   )
  ]}
 SetupSelectedActivityForIsland {nodes [
   SetupSelectedActivityForIsland() {is=title K2Node_FunctionEntry}
   > SetupIslandName() {is=title K2Node_CallFunction}
   > SetupActivityMessage() {is=title K2Node_CallFunction}
  ]}
 SetupSelectedActivityForJoinableParty {nodes [
   SetupSelectedActivityForJoinableParty() {is=title K2Node_FunctionEntry}
   > SetText!(
    execute
    self:WBP UIKit Highlight Base Object Reference=Get JP_IslandName!(self:WBP Home Bar Details Object Reference=Get WBP_HomeBarDetails()) -> (JP_IslandName:WBP UIKit Highlight Object Reference)
    Text:Text=Get RichPresenceText!(self:Discover Selected Activity View Model Object Reference=Get DiscoverSelectedActivityViewModel()) -> (RichPresenceText:Text)
   ) -> (then)
   > "SetText (Text)"(self=Get Text_SelectedTile(self=Get WBP_HomeBarDetails()), InText="LOCTEXT(\"JOIN THE PARTY\")")
  ]}
 "ShouldAutoScrollToDisco " {nodes [
   "ShouldAutoScrollToDisco "() {is=title K2Node_FunctionEntry}
   > n1 = GetShouldImmediateTransitionToHomebar() {is=title K2Node_CallFunction}
   n2 = GetOwningPlayer()
   > n3 = GetCurrentStateWidget!(execute, PlayerController:Player Controller Object Reference=@n2) -> (then, ReturnValue:Common Activatable Widget Object Reference)
   > n4 = Cast To AthenaUIStateWidget_Frontend!(execute, Object=@n3) -> (then, CastFailed, AsAthena UIState Widget Frontend:Athena UIState Widget Frontend Object Reference) {is=title K2Node_DynamicCast}
   n5 = IsLocalPlayerFreshAccount!(self:Athena UIState Widget Frontend Object Reference=@n4) -> (ReturnValue:bool)
   n6 = NOR Boolean!(
    A:bool=WillBeInSeasonFirstTimeFlow!(self:Athena UIState Widget Frontend Object Reference=@n4) -> (ReturnValue:bool)
    B:bool=IsInSeasonFirstTimeFlow!(self:Athena UIState Widget Frontend Object Reference=@n4) -> (ReturnValue:bool)
   ) -> (ReturnValue:bool)
   n7 = IsEnabledDiscoveryLandingOnSeasonFirstTimeFlow!(None:%=@n4) -> (ReturnValue:bool)
   > Return Node!(
    execute
    Autoscroll:bool=AND Boolean!(
     A=OR Boolean(A=@n6, B=@n7)
     B=NOT Boolean(A=@n5)
     C:bool=@n1
     D:bool=GetDefaultDiscoverLandingExperimentValue!(PlayerController:Player Controller Object Reference=@n2) -> (ReturnValue:bool)
    ) -> (ReturnValue)
   )
   Return Node(execute=@n4.CastFailed, Autoscroll=false)
  ]}
 TransitionToHomebar {nodes [
   TransitionToHomebar() {is=title K2Node_FunctionEntry}
   > StartDiscoverAnalyticSession(self=Get ActivityDiscoverView_NEW_VM())
   > n1 = GetShouldImmediateTransitionToHomebar()
   > n2 = Branch(Condition=@n1)
   PlayAnimationReverse(execute=[@n3.then, @n2.then], InAnimation=Get Anim_Homebar_Reveal())
   > SetPeekButtonVisibility!(execute, self:WBP Blade Object Reference=Get WBP_SlidingBlade(), HidePeekButton:bool=true) -> (then)
   > GetLobbyWidgetThenTransition!(execute, self, IsOnPlayView:bool=false) -> (then) {is=title K2Node_CallFunction}
   > UpdateMiscActionBindingVisibility(self=Get ActivityDiscoverView_NEW_VM(), bVisible=false)
   > SetBottomButtonVisibility(self=Get Button_BackToTop_BottomBarHack(), ShouldButtonBeVisible=false)
   n3 = TransitionToSecondPeekState(execute=@n2.else, self:WBP Blade Object Reference=Get WBP_SlidingBlade())
  ]}
 TransitionToPlayLobby {nodes [
   TransitionToPlayLobby() {is=title K2Node_FunctionEntry}
   > FireDiscoverExitedAnalyticEvent(self=Get ActivityDiscoverView_NEW_VM())
   > TransitionToFirstPeekState(self:WBP Blade Object Reference=Get WBP_SlidingBlade())
   > PlayAnimationForward(InAnimation=Get Anim_Homebar_Reveal())
   > SetPeekButtonVisibility(self=Get WBP_SlidingBlade(), HidePeekButton=false)
   > GetLobbyWidgetThenTransition(IsOnPlayView=true)
   > SetBottomButtonVisibility(self=Get Button_BackToTop_BottomBarHack(), ShouldButtonBeVisible=false)
   > UpdateMiscActionBindingVisibility(self=Get ActivityDiscoverView_NEW_VM(), bVisible=false)
  ]}
 UpdateOverlayBladeContentPadding {nodes [
   UpdateOverlayBladeContentPadding!() -> (then, IsContracted:bool=~) {is=title K2Node_FunctionEntry}
   > Branch(Condition=IsUsingCompactUI())
   n1 = Get WBP_SlidingBlade()
   n2 = Get bSafezoneTopEnabled()
   n3 = CompareToScreenAspectRatio!(Comparison:EAspectRatioComparison Enum=LessThan, AspectRatio:EAspectRatio Enum=Ratio_23_16) -> (ReturnValue:bool)
   n4 = SelectFloat(A=112, B=88, bPickA=@n3)
   n5 = SelectFloat(A=48, B=32, bPickA=@n3)
   > SetPadding!(
    execute
    self:Grid Slot Object Reference=SlotAsGridSlot(Widget=Get Overlay_BladeContent!(self:WBP Blade Object Reference=@n1) -> (Overlay_BladeContent:Overlay Object Reference))
    InPadding_Left:float=0
    InPadding_Top:float=SelectFloat(A=88, B=SelectFloat(A=@n4, B=@n5, bPickA=@n2), bPickA=Get IsContracted() -> (IsContracted:bool))
    InPadding_Right:float=0
    InPadding_Bottom:float=0
   ) -> (then)
  ]}
 UpdatePeekPercentageFromHomebarHeight {nodes [
   n1 = UpdatePeekPercentageFromHomebarHeight!() -> (then, HomebarHeight:double) {is=title K2Node_FunctionEntry}
   n2 = Get!(self:Grid Slot Object Reference=SlotAsGridSlot(Widget=Get ActivityDiscoverView_NEW_VM())) -> (Padding_Left:float=~, Padding_Top:float=~, Padding_Right:float=~, Padding_Bottom:float=~)
   n3 = Get(self=SlotAsGridSlot(Widget=Get CommonVisualAttachment() -> (CommonVisualAttachment:Common Visual Attachment Object Reference))) -> (Padding_Left=~, Padding_Top=~, Padding_Right=~, Padding_Bottom=~)
   > SetPeekPushPercentFromWidgetSize!(
    execute
    self:WBP Blade Object Reference=Get WBP_SlidingBlade()
    WidgetSize:double=@n1
    PercentageOfWidgetToShow:double=GetHomebarPeekPercentage!(self:Fort Discover Homespace Object Reference) -> (ReturnValue:float)
    AditionnalOffset:double=float * float(A=float + float(A=@n2.Padding_Top, B=@n3.Padding_Top), B=-1)
    PercentageOfWidgetToShowSecondary:double=1.1
   ) -> (then)
  ]}
 UpdateSlidingBladeClickCatcherAvailability {nodes [
   UpdateSlidingBladeClickCatcherAvailability() {is=title K2Node_FunctionEntry}
   n1 = IsUsingInputType!(InputType:ECommonInputType Enum=Touch, PlayerController:Player Controller Object Reference=GetOwningPlayer()) -> (ReturnValue:bool)
   n2 = GetIsDiscoverReduxEnabled()
   > SetHasClickCatcherButton!(
    execute
    self:WBP Blade Object Reference=Get WBP_SlidingBlade()
    InHasClickCatcherButton:bool=NOT Boolean(A=AND Boolean!(A=@n2, B=@n1) -> (ReturnValue))
   ) -> (then)
  ]}
 UpdateVisualForArabic {nodes [
   UpdateVisualForArabic() {is=title K2Node_FunctionEntry}
   n1 = "Equal Exactly (String)"(A=GetCurrentCulture!() -> (ReturnValue:String), B=ar)
   > Set HorizontalAlignment!(
    execute
    HorizontalAlignment:EHorizontalAlignment Enum=Select(Option 0:EHorizontalAlignment Enum=HAlign_Left, Option 1:EHorizontalAlignment Enum=HAlign_Right, Index=@n1) -> (ReturnValue:EHorizontalAlignment Enum)
    self:Grid Slot Object Reference=SlotAsGridSlot(Widget=Get CommonVisualAttachment())
   ) -> (then, Output_Get:EHorizontalAlignment Enum=HAlign_Fill)
   > Set ContentAnchor!(
    execute
    ContentAnchor:Vector 2D Structure=Select(Option 0:Vector 2D Structure=(X=0,Y=1), Option 1:Vector 2D Structure=(X=1,Y=1), Index=@n1) -> (ReturnValue:Vector 2D Structure)
    self:Common Visual Attachment Object Reference=Get CommonVisualAttachment()
   ) -> (then, Output_Get:Vector 2D Structure)
   > Set Nudge!(
    execute
    Nudge:Vector 2D Structure=Select(Option 0=(X=80,Y=-16), Option 1=(X=-80,Y=-16), Index=@n1)
    self:Grid Slot Object Reference=SlotAsGridSlot(Widget=Get CommonVisualAttachment())
   ) -> (then, Output_Get:Vector 2D Structure)
  ]}
]
variables [
 Anim_OutputBackgroundColor {type=number}
 Background MID {type=MaterialInstanceDynamic}
 BackgroundBottom MID {type=MaterialInstanceDynamic}
 BackgroundColor {type=E_Discover_Background_Color, default=Discover}
 BackgroundTop MID {type=MaterialInstanceDynamic}
 bSafezoneSidesEnabled {type=boolean}
 bSafezoneTopEnabled {type=boolean}
 CachedBackgroundColor {type=E_Discover_Background_Color, default=Discover}
 CachedTabIndex {type=integer}
 CurrentScrollOffset {type=number}
 IsShowingLibraryTile {type=boolean}
 Mask Fade Power Parameter {type=number, default=3}
 MaskFadePowerParameterMobile {type=number, default=5}
 MaskFadePowerParameterMobileSafezone {type=number, default=10}
 OutputBackgroundColor {type=LinearColor, default=<rgba 0 0 0 0>}
 OutputBackgroundSecondaryColor {type=LinearColor, default=<rgba 0 0 0 0>}
 ShouldDisplayLibraryTooltip {type=boolean, default=true}
]
widgets {
 info {
  inheritedWidgetCount=0
  namedSlotCount=2
  parentClass=/Script/DiscoveryBrowserUI.FortDiscoverHomespace
  rootWidgetClass=&GridPanel
  widgetCount=17
 }
 widgets [
  RootGrid_Homespace {widgetClassPath=&GridPanel, widget=RootGrid_Homespace, properties {columnFill [1], rowFill [1]}}
  HomeBarBacking {
   widgetClassPath=&Image
   bIsVariable=true
   parent=RootGrid_Homespace
   widget=HomeBarBacking
   properties {
    renderOpacity=0
    visibility=HitTestInvisible
    brush {
     bIsDynamicallyLoaded=false
     drawAs=Image
     imageType=FullColor
     mirroring=NoMirror
     resourceName=None
     resourceObject=/DiscoveryBrowser/TabScreens/DiscoverHome/Materials/MI_UI_HomeBar_Backing
     tiling=NoTile
     imageSize=<128 128>
     margin=<ltrb 0 0 0 0>
     outlineSettings {
      bUseBrushTransparency=false
      roundingType=HalfHeightRadius
      width=0
      color {colorUseRule=UseColor_Specified, specifiedColor=<rgba 0 0 0 0>}
      cornerRadii=<0 0 0 0>
     }
     tintColor {colorUseRule=UseColor_Specified, specifiedColor=<rgba 1 1 1 1>}
     uVRegion {bIsValid=false, max=<0 0>, min=<0 0>}
    }
    renderTransform {angle=0, scale=<1 2.5>, shear=<0 0>, translation=<0 0>}
    renderTransformPivot=<0.5 1>
   }
   slot {class=GridSlot, properties {layer=-1, verticalAlignment=VAlign_Bottom}}
  }
  TopBarGradient {
   widgetClassPath=&Image
   bIsVariable=true
   parent=RootGrid_Homespace
   widget=TopBarGradient
   properties {brush {
     bIsDynamicallyLoaded=false
     drawAs=Image
     imageType=NoImage
     mirroring=NoMirror
     resourceName=None
     resourceObject=/DiscoveryBrowser/TabScreens/DiscoverHome/Materials/MI_UI_TopBar_Gradient
     tiling=NoTile
     imageSize=<8589936640 128>
     margin=<ltrb 0 0 0 0>
     outlineSettings {
      bUseBrushTransparency=false
      roundingType=HalfHeightRadius
      width=0
      color {colorUseRule=UseColor_Specified, specifiedColor=<rgba 0 0 0 0>}
      cornerRadii=<0 0 0 0>
     }
     tintColor {colorUseRule=UseColor_Specified, specifiedColor=<rgba 1 1 1 1>}
     uVRegion {bIsValid=false, max=<0 0>, min=<0 0>}
    }}
   slot {class=GridSlot, properties {layer=-1, verticalAlignment=VAlign_Top}}
  }
  NebulousBackground {
   widgetClassPath=&Image
   bIsVariable=true
   parent=RootGrid_Homespace
   widget=NebulousBackground
   properties {
    visibility=HitTestInvisible
    brush {
     bIsDynamicallyLoaded=false
     drawAs=Image
     imageType=FullColor
     mirroring=NoMirror
     resourceName=None
     resourceObject=/FortUILibrary/MaterialInstances/Background/Nebulous/MI_UI_Background_Nebulous
     tiling=NoTile
     imageSize=<32 32>
     margin=<ltrb 0 0 0 0>
     outlineSettings {
      bUseBrushTransparency=false
      roundingType=HalfHeightRadius
      width=0
      color {colorUseRule=UseColor_Specified, specifiedColor=<rgba 0 0 0 0>}
      cornerRadii=<0 0 0 0>
     }
     tintColor {colorUseRule=UseColor_Specified, specifiedColor=<rgba 1 1 1 1>}
     uVRegion {bIsValid=false, max=<0 0>, min=<0 0>}
    }
    renderTransformPivot=<0.5 1>
   }
   slot {class=GridSlot}
  }
  WBP_SlidingBlade {
   widgetClassPath=/DiscoveryBrowser/Blade/WBP_Blade
   bIsVariable=true
   parent=RootGrid_Homespace
   widget=WBP_SlidingBlade
   properties {
    bIsClickCatcherButtonTransparent=true
    bStartExpanded=false
    eBladeAlignment=Bottom
    enableSecondaryPeekState=true
    hidePeekButton=false
    peekPushPercent=0.846
    peekPushPercentSecondary=0.666
    useFullscreenBlade=true
    backplateImageBrush {
     bIsDynamicallyLoaded=false
     drawAs=NoDrawType
     imageType=NoImage
     mirroring=NoMirror
     resourceName=None
     resourceObject=null
     tiling=NoTile
     imageSize=<32 32>
     margin=<ltrb 0 0 0 0>
     outlineSettings {
      bUseBrushTransparency=false
      roundingType=HalfHeightRadius
      width=0
      color {colorUseRule=UseColor_Specified, specifiedColor=<rgba 0 0 0 0>}
      cornerRadii=<0 0 0 1>
     }
     tintColor {colorUseRule=UseColor_Specified, specifiedColor=<rgba 1 1 1 1>}
     uVRegion {bIsValid=false, max=<0 0>, min=<0 0>}
    }
   }
   slot {class=GridSlot, properties {layer=2}}
  }
  Overlay_52 {widgetClassPath=&Overlay, namedSlotHost=WBP_SlidingBlade, widget=Overlay_52}
  Image_TopFade {
   widgetClassPath=&Image
   bIsVariable=true
   parent=Overlay_52
   widget=Image_TopFade
   properties {
    renderOpacity=0
    visibility=HitTestInvisible
    brush {
     bIsDynamicallyLoaded=false
     drawAs=Image
     imageType=FullColor
     mirroring=NoMirror
     resourceName=None
     resourceObject=/DiscoveryBrowser/TabScreens/DiscoverHome/Materials/MI_UI_Discovery_NebulousBackgroundWithTopMask
     tiling=NoTile
     imageSize=<32 128>
     margin=<ltrb 0 0 0 0>
     outlineSettings {
      bUseBrushTransparency=false
      roundingType=HalfHeightRadius
      width=0
      color {colorUseRule=UseColor_Specified, specifiedColor=<rgba 0 0 0 0>}
      cornerRadii=<0 0 0 0>
     }
     tintColor {colorUseRule=UseColor_Specified, specifiedColor=<rgba 1 1 1 1>}
     uVRegion {bIsValid=false, max=<0 0>, min=<0 0>}
    }
   }
   slot {class=OverlaySlot, properties {horizontalAlignment=HAlign_Fill}}
  }
  Image_BottomFade {
   widgetClassPath=&Image
   bIsVariable=true
   parent=Overlay_52
   widget=Image_BottomFade
   properties {
    renderOpacity=0
    visibility=HitTestInvisible
    brush {
     bIsDynamicallyLoaded=false
     drawAs=Image
     imageType=FullColor
     mirroring=NoMirror
     resourceName=None
     resourceObject=/DiscoveryBrowser/TabScreens/DiscoverHome/Materials/MI_UI_Discovery_NebulousBackgroundWithBottomMask
     tiling=NoTile
     imageSize=<32 150>
     margin=<ltrb 0 0 0 0>
     outlineSettings {
      bUseBrushTransparency=false
      roundingType=HalfHeightRadius
      width=0
      color {colorUseRule=UseColor_Specified, specifiedColor=<rgba 0 0 0 0>}
      cornerRadii=<0 0 0 0>
     }
     tintColor {colorUseRule=UseColor_Specified, specifiedColor=<rgba 1 1 1 1>}
     uVRegion {bIsValid=false, max=<0 0>, min=<0 0>}
    }
   }
   slot {class=OverlaySlot, properties {horizontalAlignment=HAlign_Fill, verticalAlignment=VAlign_Bottom}}
  }
  GridPanel_21 {widgetClassPath=&GridPanel, namedSlotHost=WBP_SlidingBlade, widget=GridPanel_21, properties {columnFill [1], rowFill [1]}}
  WBP_Discover_Tile_Library_CoachMark {
   widgetClassPath=/FortniteCoreUI/UI/DiscoveryShared/GameDetails/WBP_Discover_Tile_Library_CoachMark
   bIsVariable=true
   parent=GridPanel_21
   widget=WBP_Discover_Tile_Library_CoachMark
   slot {class=GridSlot, properties {horizontalAlignment=HAlign_Left, layer=20, verticalAlignment=VAlign_Top, nudge=<64 -256>}}
  }
  WBP_Discover_Tile_Library_DownloadStarted_CoachMark {
   widgetClassPath=/FortniteCoreUI/UI/DiscoveryShared/GameDetails/WBP_Discover_Tile_Library_DownloadStarted_CoachMark
   bIsVariable=true
   parent=GridPanel_21
   widget=WBP_Discover_Tile_Library_DownloadStarted_CoachMark
   slot {class=GridSlot, properties {horizontalAlignment=HAlign_Left, layer=19, verticalAlignment=VAlign_Top, nudge=<64 -256>}}
  }
  Wrapper_ForNavigationRuleUMGBug {
   widgetClassPath=&GridPanel
   parent=GridPanel_21
   widget=Wrapper_ForNavigationRuleUMGBug
   properties {
    navigation=&self:WidgetTree.Wrapper_ForNavigationRuleUMGBug.WidgetNavigation_0
    columnFill [1]
    rowFill [1]
   }
   slot {class=GridSlot, properties {layer=1}}
  }
  ActivityDiscoverView_NEW_VM {
   widgetClassPath=/DiscoveryBrowser/ActivityDiscoverView_NEW_VM
   bIsVariable=true
   parent=Wrapper_ForNavigationRuleUMGBug
   widget=ActivityDiscoverView_NEW_VM
   slot {class=GridSlot}
  }
  CommonVisualAttachment {
   widgetClassPath=&CommonVisualAttachment
   bIsVariable=true
   parent=GridPanel_21
   widget=CommonVisualAttachment
   properties {flowDirectionPreference=Culture, contentAnchor=<0 1>}
   slot {class=GridSlot, properties {horizontalAlignment=HAlign_Left, verticalAlignment=VAlign_Top, nudge=<80 0>, padding=<ltrb 0 120 0 0>}}
  }
  WBP_HomeBarDetails {
   widgetClassPath=&/WBP_HomeBarDetails
   bIsVariable=true
   parent=CommonVisualAttachment
   widget=WBP_HomeBarDetails
   slot {class=SizeBoxSlot}
  }
  BottomGradient {
   widgetClassPath=&Image
   bIsVariable=true
   parent=RootGrid_Homespace
   widget=BottomGradient
   properties {
    renderOpacity=0.5
    visibility=SelfHitTestInvisible
    brush {
     bIsDynamicallyLoaded=false
     drawAs=Image
     imageType=NoImage
     mirroring=NoMirror
     resourceName=None
     resourceObject=/DiscoveryBrowser/TabScreens/DiscoverHome/Materials/MI_UI_Discovery_NebulousBackgroundWithBottomMask_Black
     tiling=NoTile
     imageSize=<128 128>
     margin=<ltrb 0 0 0 0>
     outlineSettings {
      bUseBrushTransparency=false
      roundingType=HalfHeightRadius
      width=0
      color {colorUseRule=UseColor_Specified, specifiedColor=<rgba 0 0 0 0>}
      cornerRadii=<0 0 0 0>
     }
     tintColor {colorUseRule=UseColor_Specified, specifiedColor=<rgba 1 1 1 1>}
     uVRegion {bIsValid=false, max=<0 0>, min=<0 0>}
    }
    renderTransformPivot=<0.5 1>
   }
   slot {class=GridSlot, properties {layer=3, verticalAlignment=VAlign_Bottom}}
  }
  Button_BackToTop_BottomBarHack {
   widgetClassPath=/Game/UI/Frontend/BottomBar/WBP_BottomBarDecoyButton
   bIsVariable=true
   parent=RootGrid_Homespace
   widget=Button_BackToTop_BottomBarHack
   properties {triggeringInputAction {dataTable=&InputActionDataTable, rowName=BackToTop}}
   slot {class=GridSlot, properties {horizontalAlignment=HAlign_Right, layer=5, verticalAlignment=VAlign_Bottom}}
  }
 ]
}
