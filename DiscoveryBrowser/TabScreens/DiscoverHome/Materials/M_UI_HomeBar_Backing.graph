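# UI material for the Discover Home bar backing, using Modulate blend to darken what lies behind it with a vertical gradient.
# The inverted vertical normalized UV runs through an EaseInCubic curve, going from white at the top to black at the bottom.
# Vertex alpha, scaled down slightly, lerps from white (no darkening) to that gradient, so widget opacity controls the strength.
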Material /DiscoveryBrowser/TabScreens/DiscoverHome/Materials/M_UI_HomeBar_Backing
graph {
 properties {bUsedWithStaticMesh=false, blendMode=BLEND_Modulate, materialDomain=MD_UI, pixelDepthOffsetMode=PDOM_Legacy, usedShadingModels=Default Lit}
 nodes [
  n1 = ComponentMask(None=MaterialFunctionCall() {materialFunction=&GetUserInterfaceUV}.Normalized UV) {g=true}
  LinearInterpolate(
   B=MaterialFunctionCall("Value (0-1)"=OneMinus(None=@n1)) {materialFunction=&MF_UI_EaseCurves}.EaseInCubic
   Alpha=Multiply(A=VertexColor().A) {constB=0.8}
  ) {constA=1}
 ]
}
