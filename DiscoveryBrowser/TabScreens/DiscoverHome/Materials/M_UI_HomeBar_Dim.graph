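# Translucent UI material that dims the backdrop, named for the Discover home bar.
# It samples Slate post buffer 0 and feeds the color into the CoreUI BackplateC function as its post-buffer blur input.
# A Darkness parameter blends between the raw buffer color and the backplate result, and the blend is saturated.
# A separate Opacity parameter is exposed alongside it.
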Material /DiscoveryBrowser/TabScreens/DiscoverHome/Materials/M_UI_HomeBar_Dim
graph {
 properties {
  bAutomaticallySetUsageInEditor=false
  bUsedWithStaticMesh=false
  blendMode=BLEND_Translucent
  materialDomain=MD_UI
  usedShadingModels=Default Lit
 }
 nodes [
  n1 = MaterialFunctionCall!("UVs (Optional)") {materialFunction=/Engine/Functions/UserInterface/GetSlatePost0}
  Saturate(None=LinearInterpolate(
    A=@n1.LinearRGB
    B=MaterialFunctionCall!(Post Buffer Blur=@n1.LinearRGB) {materialFunction=/FortUILibrary/MaterialFunctions/CoreUI/MF_UI_CoreUI_BackplateC}.Result
    Alpha=ScalarParameter() {defaultValue=0.5, parameterName=Darkness, sliderMax=1}
   ) {constB=0})
  ScalarParameter() {defaultValue=1, parameterName=Opacity, sliderMax=1}
 ]
}
