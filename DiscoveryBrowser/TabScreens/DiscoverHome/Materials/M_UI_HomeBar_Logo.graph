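# Translucent UI material that, going by its name, draws a logo in the Discover home bar from a Texture parameter.
# It passes normalized UI UVs and the texture's size through MF_UI_UVScaling to get scaled-to-fit UVs.
# The texture is sampled with a clamped sampler, and its alpha and the fitted UVs feed MF_UI_ClipEdgePixel, whose name suggests edge-pixel clipping to avoid smearing.
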Material /DiscoveryBrowser/TabScreens/DiscoverHome/Materials/M_UI_HomeBar_Logo
graph {
 properties {
  bUsedWithStaticMesh=false
  blendMode=BLEND_Translucent
  materialDomain=MD_UI
  pixelDepthOffsetMode=PDOM_Legacy
  usedShadingModels=Default Lit
 }
 nodes [
  n1 = TextureObjectParameter() {parameterName=Texture}
  n2 = MaterialFunctionCall(
   UV=MaterialFunctionCall() {materialFunction=&GetUserInterfaceUV}.Normalized UV
   TextureSize=TextureProperty(Tex=@n1)
  ) {materialFunction=&MF_UI_UVScaling}
  MaterialFunctionCall!(UVs=@n2.Scaled to Fit, Alpha=TextureSample(UVs=@n2.Scaled to Fit, Tex=@n1) {samplerSource=SSM_Clamp_WorldGroupSettings}.A) {materialFunction=/FortUILibrary/MaterialFunctions/Basics/MF_UI_ClipEdgePixel}
 ]
}
