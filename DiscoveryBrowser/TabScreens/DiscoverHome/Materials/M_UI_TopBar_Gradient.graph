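# UI material for a top bar gradient in the Discover Home screen, which darkens what is behind it.
# It uses Modulate blending, fading vertically from the Darkness parameter at the top to white, which leaves the background unchanged, at the bottom.
# Vertex alpha blends the gradient toward white, so widget opacity fades the darkening out despite the multiplicative blend.
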Material /DiscoveryBrowser/TabScreens/DiscoverHome/Materials/M_UI_TopBar_Gradient
graph {
 properties {bUsedWithStaticMesh=false, blendMode=BLEND_Modulate, materialDomain=MD_UI, pixelDepthOffsetMode=PDOM_Legacy, usedShadingModels=Default Lit}
 nodes [
  n1 = MaterialFunctionCall() {materialFunction=&GetUserInterfaceUV}
  LinearInterpolate(
   B=LinearInterpolate(A=ScalarParameter() {defaultValue=0.6, parameterName=Darkness}, Alpha=ComponentMask(None=@n1.Normalized UV) {g=true})
   Alpha=VertexColor().A
  ) {constA=1}
 ]
}
